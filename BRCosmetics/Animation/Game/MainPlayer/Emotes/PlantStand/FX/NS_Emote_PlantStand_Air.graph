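# Niagara system for the air effect of the PlantStand player emote: a timed one-shot burst of wind meshes, glows and sparkles.
# System-level skeletal mesh sampling positions all six emitters.
# Rotating airflow meshes, one set rising, fade through dynamic material parameters and alpha curves, alongside flare, dark glow, spike and curl-noise sparkle sprites.
# Brightness or opacity scales with time of day, and some emitters add front-end versus gameplay multipliers.
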NiagaraSystem /BRCosmetics/Animation/Game/MainPlayer/Emotes/PlantStand/FX/NS_Emote_PlantStand_Air
dependencies {
 usedDataInterfaces [
  &NiagaraDataInterfaceArrayFloat
  &NiagaraDataInterfaceCurve
  &NiagaraDataInterfaceVectorCurve
  &NiagaraDataInterfaceVector2DCurve
  &NiagaraDataInterfaceSkeletalMesh
 ]
 usedDynamicInputs []
 usedModules [
  &self:SystemSpawnScript.SystemScriptSource.SystemScriptGraph.NiagaraNodeAssignment_0.SetVariables_#1
  &SystemState
  &SampleSkeletalMesh
 ]
}
emitters {
 Air_Airflow_Start {
  emitterUpdateScript [
   &EmitterState(Life Cycle Mode=::NewEnumerator1, Loop Behavior=::NewEnumerator1, Loop Delay=0.03, Loop Duration=1, UseLoopDelay=false)
   &SpawnBurst_Instantaneous(Spawn Count=6, Spawn Time=0.8, Use Spawn Probability=false, UseLoopCountLimit=false)
  ]
  particleSpawnScript [
   &InitializeParticle(
    Color Maximum=<rgba 0.4 0.47 0.5 0.1>
    Color Minimum=<rgba 0.8 0.9367 1 0.3>
    Color Mode=::NewEnumerator2
    Lifetime Max=1.5
    Lifetime Mode=::NewEnumerator1
    Mesh Scale=&Multiply_VectorByFloat(
     Vector=&RandomRangeVector(Minimum=<1 1 1.5>, Maximum=<1 1 2>)
     Float=&RandomRangeFloat(Minimum=0.3, Maximum=1)
    )
    Mesh Scale Mode=::NewEnumerator4
    Position Offset=$System.SampleSkeletalMesh.SampledPosition
    UsePositionOffset=true
    Write Material Random=true
   )
   &BoxLocation(Box Offset=<0 0 -120>, Box Size=<0 0 50>, MaskSpawnGroup=false, Module.Override Seed=false, Rotate Box=false, Surface Only=false)
   &InitialMeshOrientation(
    Mesh Orientation Mode=::NewEnumerator2
    Orientation Coordinate Space=::Local
    Rotation=&Multiply_VectorByFloat(Vector=&RandomRangeVector(Minimum=<-1 -1 -5>, Maximum=<1 1 5>), Float=0.05)
    Use Rotation Vector=true
   )
   &AddVelocity(Velocity Speed Scale=&RandomRangeFloat(Minimum=1, Maximum=2))
  ]
  particleUpdateScript [
   &ParticleState()
   &DynamicMaterialParameters(
    Index 0 Param 1=&FloatFromCurve(FloatCurve=DataInterface() {propertyValues {Curve {keys [
         {interpMode=RCIM_Cubic, tangentMode=RCTM_User, time=0, value=1}
         <at 0.7 0.35>
         {interpMode=RCIM_Cubic, tangentMode=RCTM_User, time=1, value=0}
        ]}}})
    Module.Param0WriteEnabled=true
    Module.Param1WriteEnabled=true
    Module.Param2WriteEnabled=true
    Module.Param3WriteEnabled=true
   )
   &ScaleMeshSize(Scale Factor=&Multiply_VectorByFloat(
     Vector=&Multiply_VectorByFloat(Vector=$Engine.Owner.Scale, Float=1)
     Float=&FloatFromCurve(
      FloatCurve=DataInterface() {propertyValues {Curve {keys [
          {interpMode=RCIM_Cubic, tangentMode=RCTM_User, time=0, value=0}
          {arriveTangent=1.1693, interpMode=RCIM_Cubic, leaveTangent=1.1693, tangentMode=RCTM_User, time=1, value=1}
         ]}}}
      Scale Curve=&FN_FrontEndFloatModifier(Gameplay=2.6, Front End=2.6)
     )
    ))
   &UpdateMeshOrientation(
    Rotation Rate=&SelectFloatFromArray(
     Array Sampling Mode=::NewEnumerator0
     Direct Array Index=?
     Float Selection Array=DataInterface() {propertyValues {FloatData [-1,-2]}}
     Random Seed=0
    )
    Rotation Vector=<0 0 1>
   )
   &SolveForcesAndVelocity(Clamp Velocity=false, Limit Acceleration=false)
   &ScaleColor(
    Scale Alpha=&FloatFromCurve(
     FloatCurve=DataInterface() {propertyValues {Curve {keys [
         {arriveTangent=4.1785, interpMode=RCIM_Cubic, leaveTangent=4.1785, tangentMode=RCTM_User, time=0, value=0}
         {arriveTangent=0.9995, interpMode=RCIM_Cubic, leaveTangent=0.9995, tangentMode=RCTM_User, time=0.3, value=0.625}
         {arriveTangent=-0.0405, interpMode=RCIM_Cubic, leaveTangent=-0.0405, tangentMode=RCTM_User, time=1, value=1}
        ]}}}
     Scale Curve=&FloatFromCurve(FloatCurve=DataInterface() {propertyValues {Curve {keys [
          {tangentMode=RCTM_User, time=0, value=1}
          {arriveTangent=-0.0405, interpMode=RCIM_Cubic, leaveTangent=-0.0405, tangentMode=RCTM_User, time=0.5, value=1}
          {arriveTangent=-3.6691, interpMode=RCIM_Cubic, leaveTangent=-3.6691, tangentMode=RCTM_User, time=1, value=0}
         ]}}})
    )
    Scale RGB=&Multiply_VectorByFloat(
     Vector=&Multiply_VectorByFloat(Vector=<1 1 1>, Float=&FN_FrontEndFloatModifier(Gameplay=1, Front End=1))
     Float=&FNTimeOfDayScalar(Night Scale=0.5, Morning Scale=0.3, Evening Scale=0.3)
    )
    ScaleA=true
    ScaleRGB=true
   )
  ]
  rendererClasses [&NiagaraMeshRendererProperties]
  renderers [{rendererClass=&NiagaraMeshRendererProperties, rendererIndex=0}]
 }
 Airflow_Up {
  emitterUpdateScript [
   &EmitterState(Life Cycle Mode=::NewEnumerator1, Loop Behavior=::NewEnumerator1, Loop Delay=0.85, Loop Duration=2, UseLoopDelay=true)
   &SpawnBurst_Instantaneous(Spawn Count=2, Use Spawn Probability=false, UseLoopCountLimit=false)
  ]
  particleSpawnScript [
   &InitializeParticle(
    Color=<rgba 1.05 1.3224 1.5 1>
    Lifetime=1.2
    Mesh Scale=&RandomRangeVector(Minimum=<1 1 0.8>, Maximum=<1.5 1.5 1.2>)
    Mesh Scale Mode=::NewEnumerator4
    Position Offset=&AddVectorToPosition(Position=Linked() {linkedVariable {name=System.SampleSkeletalMesh.SampledPosition, type=Vector3f}}, Vector=<0 0 -120>)
    UsePositionOffset=true
    Write Material Random=true
   )
   &FN_SetSpawnGroup(Number of Spawn Groups=2)
   &InitialMeshOrientation(
    Mesh Axis To Orient=<0 0 1>
    Mesh Orientation Mode=::NewEnumerator2
    Orientation Vector=<0 0 1>
    Rotation=&MakeVector(
     X=0
     Y=0
     Z=&SelectFloatFromArray(Direct Array Index=&SpawnGroup(), Float Selection Array=DataInterface() {propertyValues {FloatData [0,0.5]}})
    )
    Use Rotation Vector=true
   )
   &AddVelocity(Velocity=<0 0 200>)
  ]
  particleUpdateScript [
   &ParticleState()
   &DynamicMaterialParameters(
    Index 0 Param 1=&FloatFromCurve(FloatCurve=DataInterface() {propertyValues {Curve {keys [
         {arriveTangent=-1.5522, interpMode=RCIM_Cubic, leaveTangent=-1.5522, tangentMode=RCTM_User, time=0, value=1}
         {arriveTangent=-0.9602, interpMode=RCIM_Cubic, leaveTangent=-0.9602, tangentMode=RCTM_User, time=1, value=0}
        ]}}})
    Module.Param0WriteEnabled=true
    Module.Param1WriteEnabled=true
    Module.Param2WriteEnabled=true
    Module.Param3WriteEnabled=true
   ) {module=DynamicMaterialParameters001}
   &UpdateMeshOrientation(Rotation Rate=-3, Rotation Vector=<0 0 1>)
   &ScaleMeshSize(Scale Factor=&Multiply_VectorByFloat(
     Vector=&VectorFromCurve(VectorCurve=DataInterface() {propertyValues {
        XCurve {keys [{arriveTangent=0.7681, interpMode=RCIM_Cubic, leaveTangent=0.7681, tangentMode=RCTM_User, time=0, value=0.75}, <at 1 1.15 RCIM_Cubic>]}
        YCurve {keys [{arriveTangent=0.7681, interpMode=RCIM_Cubic, leaveTangent=0.7681, tangentMode=RCTM_User, time=0, value=0.75}, <at 1 1.15 RCIM_Cubic>]}
        ZCurve {keys [<at 0 0.5>,<at 1 1>]}
       }})
     Float=1.5
    ))
   &ScaleColor(
    Scale Alpha=&FloatFromCurve(FloatCurve=DataInterface() {propertyValues {Curve {keys [
         {interpMode=RCIM_Cubic, tangentMode=RCTM_User, time=0, value=0}
         {interpMode=RCIM_Cubic, tangentMode=RCTM_User, time=0.5, value=1}
         {tangentMode=RCTM_User, time=1, value=0}
        ]}}})
    Scale RGB=&Multiply_VectorByFloat(
     Vector=&Multiply_VectorByFloat(Vector=<1 1 1>, Float=&FN_FrontEndFloatModifier(Gameplay=1, Front End=1))
     Float=&FNTimeOfDayScalar(Night Scale=0.7, Morning Scale=0.3, Evening Scale=0.3)
    )
    ScaleA=true
    ScaleRGB=true
   )
   &SolveForcesAndVelocity(Clamp Velocity=false, Limit Acceleration=false)
   /SetVariables(Particles.New Linear Color:LinearColor=&MakeLinearColorFromVectorAndFloat(
     "Vector (RGB)"=&MakeVectorFromLinearColorRGB(Color=<rgba 1.05 1.3224 1.5 1>)
     "Float (Alpha)"=&Multiply_Float(
      A=&Multiply_Float(
       A=&FN_FrontEndFloatModifier(Gameplay=0.7, Front End=1.4)
       B=&FNTimeOfDayScalar(Night Scale=0.7, Morning Scale=0.3, Evening Scale=0.3)
      )
      B=&FloatFromCurve(FloatCurve=DataInterface() {propertyValues {Curve {keys [
           {interpMode=RCIM_Cubic, tangentMode=RCTM_User, time=0, value=0}
           {interpMode=RCIM_Cubic, tangentMode=RCTM_User, time=0.5, value=1}
           {tangentMode=RCTM_User, time=1, value=0}
          ]}}})
     )
    )) {isSetParameters=true}
   /SetVariables(Particles.New Vector 4f:Vector4f=&MakeVector4(
     X=&FloatFromCurve(FloatCurve=DataInterface() {propertyValues {Curve {keys [
          {arriveTangent=-1.5522, interpMode=RCIM_Cubic, leaveTangent=-1.5522, tangentMode=RCTM_User, time=0, value=1}
          {arriveTangent=-1.5926, interpMode=RCIM_Cubic, leaveTangent=-1.5926, tangentMode=RCTM_User, time=1, value=0}
         ]}}})
     Y=0
     Z=0
     W=0
    )) {isSetParameters=true}
  ]
  rendererClasses [&NiagaraMeshRendererProperties]
  renderers [
   {rendererClass=&NiagaraMeshRendererProperties, rendererIndex=0}
   {rendererClass=&NiagaraMeshRendererProperties, rendererIndex=1}
  ]
 }
 Dark_Glow {
  emitterUpdateScript [
   &EmitterState(Life Cycle Mode=::NewEnumerator1, Loop Behavior=::NewEnumerator1, Loop Delay=0.1, Loop Duration=1, UseLoopDelay=false)
   &SpawnBurst_Instantaneous(Spawn Count=1, Spawn Time=0.85, Use Spawn Probability=false)
  ]
  particleSpawnScript [
   &InitializeParticle(
    Color=<rgba 0.03 0.03 0.03 1>
    Lifetime=0.4
    Position:Vector3f=$System.SampleSkeletalMesh.SampledPosition
    Position Mode=::NewEnumerator1
    Position Offset=<0 0 -10>
    Sprite Rotation Mode=::NewEnumerator2
    Sprite Size Mode=::NewEnumerator3
    Uniform Sprite Size=500
    UsePositionOffset=true
    Write Material Random=true
   )
   &DynamicMaterialParameters(
    Index 0 Param 1=&RandomRangeFloat(Minimum=0.4, Maximum=0.6)
    Index 0 Param 2=1
    Index 0 Param 3=1
    Index 0 Param 4=1
    Module.Param0WriteEnabled=true
    Module.Param1WriteEnabled=true
    Module.Param2WriteEnabled=true
    Module.Param3WriteEnabled=true
   )
   &CameraOffset(Camera Offset Amount=-10)
  ]
  particleUpdateScript [
   &ParticleState()
   &SolveForcesAndVelocity(Clamp Velocity=false, Limit Acceleration=false)
   &ScaleColor(
    Scale Alpha=&Multiply_Float(
     A=1
     B=&FloatFromCurve(FloatCurve=DataInterface() {propertyValues {Curve {keys [<at 0 0>,<at 0.2 1>,<at 0.8 1>,<at 1 0>]}}})
    )
    Scale RGB=&Multiply_VectorByFloat(Vector=<1 1 1>, Float=&FNTimeOfDayScalar(Night Scale=0.5, Morning Scale=0.3, Evening Scale=0.4))
    ScaleA=true
    ScaleRGB=true
   )
   &ScaleSpriteSize(Scale Factor=&Vector2DFromFloat(Value=&FloatFromCurve(FloatCurve=DataInterface() {propertyValues {Curve {keys [<at 0 1.5 RCIM_Cubic>,<at 1 0>]}}})))
  ]
  rendererClasses [&NiagaraSpriteRendererProperties]
  renderers [{rendererClass=&NiagaraSpriteRendererProperties}]
 }
 Explode_Sparkles {
  emitterUpdateScript [
   &EmitterState(Life Cycle Mode=::NewEnumerator1, Loop Behavior=::NewEnumerator1, Loop Delay=0.1, Loop Duration=1, UseLoopDelay=false)
   &SpawnBurst_Instantaneous(Spawn Count=30, Spawn Time=0.85, Use Spawn Probability=false, UseLoopCountLimit=false)
  ]
  particleSpawnScript [
   &InitializeParticle(
    Color=<rgba 1.4 1.6631 2 1>
    Lifetime=&RandomRangeFloat(Minimum=0.3, Maximum=0.8)
    Position:Vector3f=$System.SampleSkeletalMesh.SampledPosition
    Position Mode=::NewEnumerator1
    Position Offset=<0 0 -10>
    Sprite Rotation Mode=::NewEnumerator0
    Sprite Size Max=<1 5>
    Sprite Size Min=<1 3>
    Sprite Size Mode=::NewEnumerator2
    UsePositionOffset=true
    Write Material Random=true
   )
   /SetVariables(Particles.RandomCore:NiagaraFloat=&FloatFromCurve(
     FloatCurve=DataInterface() {propertyValues {Curve {keys [
         {arriveTangent=0.0559, interpMode=RCIM_Cubic, leaveTangent=0.0559, tangentMode=RCTM_User, time=0, value=0.2}
         {arriveTangent=0.1895, interpMode=RCIM_Cubic, leaveTangent=0.1895, tangentMode=RCTM_User, time=0.9, value=0.25}
         {arriveTangent=1.1808, interpMode=RCIM_Cubic, leaveTangent=1.1808, tangentMode=RCTM_User, time=1, value=0.3}
        ]}}}
     CurveIndex=&RandomRangeFloat(Minimum=0, Maximum=1)
     Scale Curve=2
    )) {isSetParameters=true}
   &ShapeLocation(MaskSpawnGroup=false, Override Seed=false, Sphere Radius=20, UseOwnerScale=true)
   &AddVelocity(
    Module.Use Falloff Distance=false
    Override Seed=false
    Velocity Mode=::NewEnumerator1
    Velocity Origin=$Particles.Position
    Velocity Speed=&RandomRangeFloat(Minimum=100, Maximum=200)
   )
  ]
  particleUpdateScript [
   /SetVariables(Particles.Glow03_Color=&MakeLinearColorFromVectorAndFloat(
     "Vector (RGB)"=&MakeVectorFromLinearColorRGB(Color=&MultiplyLinearColor(
       Linear Color=<rgba 1 1 1 0.2>
       Scale Factor=&MakeVector4_FromFloat(Float=&FNTimeOfDayScalar(Night Scale=0.5, Morning Scale=0.2, Evening Scale=0.2))
      ))
     "Float (Alpha)"=&Multiply_Float(A=0.1, B=1)
    )) {isSetParameters=true}
   /SetVariables(Particles.Glow03_Size=&Multiply_Vector2D_ByFloat(
     Float=&FloatFromCurve(
      FloatCurve=DataInterface() {propertyValues {Curve {keys [
          <at 0 0 RCIM_Cubic>
          <at 0.2 1.3 RCIM_Cubic>
          {arriveTangent=-0.1423, interpMode=RCIM_Cubic, leaveTangent=-0.1423, tangentMode=RCTM_User, time=1, value=0}
         ]}}}
      Scale Curve=10
     )
     Vector2D=<1 2>
    )) {isSetParameters=true}
   &ParticleState()
   &CurlNoiseForce(Noise Strength=&Multiply_Float(A=2000, B=1), Pan Noise=true, Pan Noise Field=<0 0 1>)
   &ScaleSpriteSize(Scale Factor=&Vector2DFromFloat(Value=&FloatFromCurve(
      FloatCurve=DataInterface() {propertyValues {Curve {keys [
          <at 0 0 RCIM_Cubic>
          <at 0.2 1.2 RCIM_Cubic>
          {arriveTangent=-2.7132, interpMode=RCIM_Cubic, leaveTangent=-2.7132, tangentMode=RCTM_User, time=1, value=0}
         ]}}}
      Scale Curve=0.5
     )))
   &ScaleColor(
    Scale Alpha=&Multiply_Float(A=1, B=1)
    Scale RGB=&Multiply_VectorByFloat(Vector=<2 2 2>, Float=&FNTimeOfDayScalar(Night Scale=1, Morning Scale=0.7, Evening Scale=0.7))
    ScaleA=true
    ScaleRGB=true
   )
   &Drag(Drag=2, Use Linear Drag=true, Use Rotational Drag=false)
   &SolveForcesAndVelocity(Clamp Velocity=false, Limit Acceleration=false)
  ]
  rendererClasses [&NiagaraSpriteRendererProperties]
  renderers [
   {rendererClass=&NiagaraSpriteRendererProperties}
   {rendererClass=&NiagaraSpriteRendererProperties, rendererIndex=1}
  ]
 }
 Flare {
  emitterUpdateScript [
   &EmitterState(Life Cycle Mode=::NewEnumerator1, Loop Behavior=::NewEnumerator1, Loop Duration=1, UseLoopDelay=false)
   &SpawnBurst_Instantaneous(Spawn Count=1, Spawn Time=0.85, Use Spawn Probability=false)
  ]
  particleSpawnScript [
   &InitializeParticle(
    Color=<rgba 1.05 1.3947 1.5 1>
    Lifetime=0.4
    Position=$System.SampleSkeletalMesh.SampledPosition
    Position Mode=::NewEnumerator1
    Position Offset=<0 0 -10>
    Sprite Rotation Mode=::NewEnumerator2
    Sprite Size Mode=::NewEnumerator3
    Uniform Sprite Size=600
    UsePositionOffset=true
    Write Material Random=true
   )
   &DynamicMaterialParameters(
    Index 0 Param 1=&RandomRangeFloat(Minimum=0.4, Maximum=0.6)
    Index 0 Param 2=1
    Index 0 Param 3=1
    Index 0 Param 4=1
    Module.Param0WriteEnabled=true
    Module.Param1WriteEnabled=true
    Module.Param2WriteEnabled=true
    Module.Param3WriteEnabled=true
   )
   &CameraOffset()
  ]
  particleUpdateScript [
   &ParticleState()
   &SolveForcesAndVelocity(Clamp Velocity=false, Limit Acceleration=false)
   &ScaleColor(
    Scale Alpha=&Multiply_Float(
     A=&FNTimeOfDayScalar(Night Scale=0.5, Morning Scale=0.3, Evening Scale=0.4)
     B=&FloatFromCurve(FloatCurve=DataInterface() {propertyValues {Curve {keys [<at 0 0>,<at 0.2 1>,<at 0.8 1>,<at 1 0>]}}})
    )
    Scale RGB=&Multiply_VectorByFloat(Vector=<1 1 1>, Float=&FN_FrontEndFloatModifier(Gameplay=1, Front End=1.2))
    ScaleA=true
    ScaleRGB=true
   )
   &ScaleSpriteSize(Scale Factor=&Vector2DFromFloat(Value=&FloatFromCurve(FloatCurve=DataInterface() {propertyValues {Curve {keys [<at 0 1.5 RCIM_Cubic>,<at 1 0>]}}})))
  ]
  rendererClasses [&NiagaraSpriteRendererProperties]
  renderers [{rendererClass=&NiagaraSpriteRendererProperties}]
 }
 Spikes {
  emitterUpdateScript [
   &EmitterState(Life Cycle Mode=::NewEnumerator1, Loop Behavior=::NewEnumerator1, Loop Delay=0.03, Loop Duration=1, UseLoopDelay=true)
   &SpawnBurst_Instantaneous(Spawn Count=6, Spawn Time=0.85, Use Spawn Probability=false, UseLoopCountLimit=false)
  ]
  particleSpawnScript [
   &InitializeParticle(
    Color=<rgba 1.05 1.3224 1.5 0.1>
    Lifetime Max=0.3
    Lifetime Min=0.15
    Lifetime Mode=::NewEnumerator1
    Position=$System.SampleSkeletalMesh.SampledPosition
    Position Mode=::NewEnumerator1
    Position Offset=<0 0 -130>
    Sprite Size Max=<60 150>
    Sprite Size Min=<45 75>
    Sprite Size Mode=::NewEnumerator2
    UsePositionOffset=true
    Write Material Random=true
   )
   &TorusLocation(Align Torus With Axis=false, Handle Radius=0, Large Radius=10, MaskSpawnGroup=false, Module.Override Randomness=false)
   &AddVelocityFromPoint(
    Module.Use Falloff Distance=false
    Origin Offset=&AddVectorToPosition(Position=$System.SampleSkeletalMesh.SampledPosition, Vector=<0 0 -140>)
    Velocity Strength=&RandomRangeFloat(Minimum=50, Maximum=100)
   )
   &SubUVAnimation(End Frame=0, SubUV Animation Mode:ENiagara_SubUVLookupMode=::NewEnumerator2)
  ]
  particleUpdateScript [
   &ParticleState()
   &Drag(Drag=6, Use Linear Drag=true, Use Rotational Drag=false)
   &SolveForcesAndVelocity(Clamp Velocity=false, Limit Acceleration=false)
   &ScaleSpriteSize(
    Non-Uniform Curve Scale=&Multiply_Vector2D_ByFloat(Float=0.6, Vector2D=<1 1.2>)
    Non-Uniform Curve Sprite Scale=DataInterface() {propertyValues {
      XCurve {keys [{tangentMode=RCTM_User, time=0, value=1}, {tangentMode=RCTM_User, time=1, value=0.8}]}
      YCurve {keys [{tangentMode=RCTM_User, time=0, value=0.5}, {tangentMode=RCTM_User, time=1, value=1.7}]}
     }}
    Scale Sprite Size Mode=::NewEnumerator3
   )
   &ScaleColor(
    Scale Alpha=&FloatFromCurve(
     FloatCurve=DataInterface() {propertyValues {Curve {keys [<at 0 0>,<at 0.1 1>,<at 0.2 0.8>,<at 1 0>]}}}
     Scale Curve=&FNTimeOfDayScalar(Night Scale=1, Morning Scale=0.3, Evening Scale=0.5)
    )
    ScaleA=true
    ScaleRGB=true
   )
  ]
  rendererClasses [&NiagaraSpriteRendererProperties]
  renderers [{rendererClass=&NiagaraSpriteRendererProperties}]
 }
}
summary {
 systemName=NS_Emote_PlantStand_Air
 emitters [
  {emitterName=Airflow_Up, rendererClasses [&NiagaraMeshRendererProperties]}
  {emitterName=Explode_Sparkles, rendererClasses [&NiagaraSpriteRendererProperties]}
  {emitterName=Flare, rendererClasses [&NiagaraSpriteRendererProperties]}
  {emitterName=Dark_Glow, rendererClasses [&NiagaraSpriteRendererProperties]}
  {emitterName=Spikes, rendererClasses [&NiagaraSpriteRendererProperties]}
  {emitterName=Air_Airflow_Start, rendererClasses [&NiagaraMeshRendererProperties]}
 ]
 userVariables [
  User.bIsFrontEnd {type=NiagaraBool}
  User.bIsFrontEndPreview {type=NiagaraBool}
 ]
}
