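# Niagara effect for the earth part of the PlantStand player emote, with swirling dust, floating rocks and break-up debris.
# Ribbon trail and smoke orbit a sampled skeletal mesh point, while seven toon-outlined rocks grow and spin on indexed bones or sockets.
# Two flare emitters track the rocks through particle reads; rocks emit location and death events, alongside debris, smoke and spark emitters.
# Time of day scales colour and opacity, with platform-dependent outline width.
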NiagaraSystem /BRCosmetics/Animation/Game/MainPlayer/Emotes/PlantStand/FX/NS_Emote_PlantStand_Earth
dependencies {
 usedDataInterfaces [
  &NiagaraDataInterfacePlatformSet
  &NiagaraDataInterfaceCurve
  &NiagaraDataInterfaceArrayFloat
  &NiagaraDataInterfaceParticleRead
  &NiagaraDataInterfaceSkeletalMesh
 ]
 usedDynamicInputs []
 usedModules [
  &self:SystemSpawnScript.SystemScriptSource.SystemScriptGraph.NiagaraNodeAssignment_0.SetVariables_#1
  &SystemState
  &SampleSkeletalMesh
 ]
}
emitters {
 DryIceSmoke {
  emitterUpdateScript [&EmitterState(Life Cycle Mode=::NewEnumerator1, Loop Behavior=::NewEnumerator1, Loop Duration=6, UseLoopDelay=false)]
  particleSpawnScript [
   &InitializeParticle(
    Color Maximum=<rgba 0.2 0.15 0.1 1>
    Color Minimum=<rgba 0.3 0.23 0.18 1.5>
    Color Mode=::NewEnumerator2
    Lifetime=&RandomRangeFloat(Minimum=0.3, Maximum=0.5)
    Sprite Rotation Mode=::NewEnumerator2
    Sprite Size Max=<60 120>
    Sprite Size Min=<50 100>
    Sprite Size Mode=::NewEnumerator2
    Sprite UV Mode=::NewEnumerator3
    UsePositionOffset=false
    Write Material Random=true
   )
   &CameraOffset(Camera Offset Amount=-5)
  ]
  particleUpdateScript [
   &ParticleState()
   &ScaleSpriteSize(
    Uniform Curve Scale=1.25
    Uniform Curve Sprite Scale=DataInterface() {propertyValues {Curve {keys [
        {arriveTangent=2.1534, interpMode=RCIM_Cubic, leaveTangent=2.1534, tangentMode=RCTM_User, time=0, value=0.25}
        {interpMode=RCIM_Cubic, tangentMode=RCTM_User, time=1, value=1.4}
       ]}}}
   )
   &ScaleColor(
    Scale Alpha=&FloatFromCurve(
     FloatCurve=DataInterface() {propertyValues {Curve {keys [
         {arriveTangent=7.6695, interpMode=RCIM_Cubic, leaveTangent=7.6695, tangentMode=RCTM_User, time=0, value=0}
         {interpMode=RCIM_Cubic, tangentMode=RCTM_User, time=0.4, value=1}
         {arriveTangent=-1.1661, interpMode=RCIM_Cubic, leaveTangent=-1.1661, tangentMode=RCTM_User, time=1, value=0}
        ]}}}
     Scale Curve=&FNTimeOfDayScalar(Night Scale=0.5, Morning Scale=0.75, Evening Scale=0.75)
    )
    ScaleA=true
    ScaleRGB=true
   )
   &GravityForce(Gravity=<0 0 -100>)
   &Drag(Drag=&RandomRangeFloat(Minimum=6, Maximum=10), Use Linear Drag=true, Use Rotational Drag=true)
   &SolveForcesAndVelocity(Clamp Velocity=false, Limit Acceleration=false)
  ]
  rendererClasses [&NiagaraSpriteRendererProperties]
  renderers [{rendererClass=&NiagaraSpriteRendererProperties}]
 }
 Earth_Explode_Sparks_Death_Event {
  emitterUpdateScript [
   &EmitterState(Life Cycle Mode=::NewEnumerator1, Loop Behavior=::NewEnumerator1, Loop Delay=0.1, Loop Duration=6, UseLoopDelay=false)
   &SpawnBurst_Instantaneous(Spawn Count=30, Spawn Time=4.9, Use Spawn Probability=false, UseLoopCountLimit=false)
  ]
  particleSpawnScript [
   &InitializeParticle(
    Color Maximum=<rgba 1 0.636 0.4428 0.6>
    Color Minimum=<rgba 1 0.3293 0.1562 1>
    Color Mode=::NewEnumerator2
    Lifetime=&RandomRangeFloat(Minimum=0.3, Maximum=0.7)
    Sprite Rotation Mode=::NewEnumerator0
    Sprite Size Max=<1.2 5>
    Sprite Size Min=<0.75 3>
    Sprite Size Mode=::NewEnumerator2
    UsePositionOffset=true
    Write Material Random=true
   )
   /SetVariables(Particles.RandomCore:NiagaraFloat=&FloatFromCurve(
     FloatCurve=DataInterface() {propertyValues {Curve {keys [
         {arriveTangent=0.0559, interpMode=RCIM_Cubic, leaveTangent=0.0559, tangentMode=RCTM_User, time=0, value=0.2}
         {arriveTangent=0.1895, interpMode=RCIM_Cubic, leaveTangent=0.1895, tangentMode=RCTM_User, time=0.9, value=0.25}
         {arriveTangent=1.1808, interpMode=RCIM_Cubic, leaveTangent=1.1808, tangentMode=RCTM_User, time=1, value=0.3}
        ]}}}
     CurveIndex=&RandomRangeFloat(Minimum=0, Maximum=1)
     Scale Curve=2
    )) {isSetParameters=true}
  ]
  particleUpdateScript [
   &ParticleState()
   &CurlNoiseForce(Noise Strength=350, Pan Noise=true, Pan Noise Field=<0 0 0.1>)
   &ScaleSpriteSize(Scale Factor=&Vector2DFromFloat(Value=&FloatFromCurve(FloatCurve=DataInterface() {propertyValues {Curve {keys [
          {interpMode=RCIM_Cubic, tangentMode=RCTM_User, time=0, value=1}
          {arriveTangent=-2.709, interpMode=RCIM_Cubic, leaveTangent=-2.709, tangentMode=RCTM_User, time=1, value=0}
         ]}}})))
   /SetVariables(Particles.Glow03_Size=&Multiply_Vector2D(A=$Particles.SpriteSize, B=<4 3>)) {isSetParameters=true}
   &ScaleColor(
    Scale Alpha=&FloatFromCurve(FloatCurve=DataInterface() {propertyValues {Curve {keys [
         {interpMode=RCIM_Cubic, tangentMode=RCTM_User, time=0.5, value=1}
         {arriveTangent=-2.9886, interpMode=RCIM_Cubic, leaveTangent=-2.9886, tangentMode=RCTM_User, time=1, value=0}
        ]}}})
    Scale RGB=&Multiply_VectorByFloat(Vector=<2 2 2>, Float=&FNTimeOfDayScalar(Night Scale=1, Morning Scale=0.7, Evening Scale=0.7))
    ScaleA=true
    ScaleRGB=true
   )
   /SetVariables(Particles.Glow03_Color=&MultiplyLinearColor(
     Linear Color=$Particles.Color
     Scale Factor=&MakeFloat4FromVectorAndFloat("Vector (RGB)"=<1.5 1.5 1.5>, "Float (Alpha)"=0.1)
    )) {isSetParameters=true}
   &Drag(
    Drag=&FloatFromCurve(
     FloatCurve=DataInterface() {propertyValues {Curve {keys [
         {arriveTangent=1.5108, interpMode=RCIM_Cubic, leaveTangent=1.5108, tangentMode=RCTM_User, time=0, value=0.5}
         {interpMode=RCIM_Cubic, tangentMode=RCTM_User, time=1, value=1}
        ]}}}
     Scale Curve=10
    )
    Use Linear Drag=true
    Use Rotational Drag=false
   )
   &SolveForcesAndVelocity(Clamp Velocity=false, Limit Acceleration=false)
   &ScaleSpriteSizeBySpeed(Max Scale Factor=<0.3 5>)
  ]
  rendererClasses [&NiagaraSpriteRendererProperties]
  renderers [
   {rendererClass=&NiagaraSpriteRendererProperties}
   {rendererClass=&NiagaraSpriteRendererProperties, rendererIndex=1}
  ]
 }
 Earth_Spawn_Smoke {
  emitterUpdateScript [
   &EmitterState(Life Cycle Mode=::NewEnumerator1, Loop Behavior=::NewEnumerator1, Loop Duration=1.7, UseLoopDelay=false)
   &SpawnRate(Spawn Probability=0.75, SpawnRate=20, Use Spawn Probability=true)
  ]
  particleSpawnScript [
   &InitializeParticle(
    Color=&MakeLinearColorFromVectorAndFloat(
     "Vector (RGB)"=&MakeVectorFromLinearColorRGB(Color=<rgba 0.3 0.23 0.18 2>)
     "Float (Alpha)"=&FloatFromCurve(
      FloatCurve=DataInterface() {propertyValues {Curve {keys [
          {arriveTangent=9.3399, interpMode=RCIM_Cubic, leaveTangent=9.3399, tangentMode=RCTM_User, time=0, value=0}
          {interpMode=RCIM_Cubic, tangentMode=RCTM_User, time=0.3, value=1}
          {tangentMode=RCTM_User, time=0.8, value=0}
         ]}}}
      CurveIndex=$Emitter.NormalizedLoopAge
     )
    )
    Lifetime=0.7
    Sprite Rotation Mode=::NewEnumerator2
    Sprite Size Mode=::NewEnumerator1
    Uniform Sprite Size Max=50
    Uniform Sprite Size Min=30
    UsePositionOffset=false
    Write Material Random=true
   ) {module=InitializeParticle001}
   &DynamicMaterialParameters(Module.Index0Param0Write=true, Module.Index0Param1Write=true, Module.Index0Param2Write=true, Module.Index0Param3Write=true)
   &RotateAroundPoint(
    Radius:NiagaraFloat=&FloatFromCurve(
     FloatCurve=DataInterface() {propertyValues {Curve {keys [{interpMode=RCIM_Cubic, tangentMode=RCTM_User, time=0, value=0.875}, {interpMode=RCIM_Cubic, tangentMode=RCTM_User, time=1, value=1.125}]}}}
     CurveIndex=$Emitter.NormalizedLoopAge
     Scale Curve=60
    )
    Rotation Center:NiagaraPosition=$System.SampleSkeletalMesh.SampledPosition
    Rotation Phase:NiagaraFloat=&Multiply_Float(A=$Emitter.NormalizedLoopAge, B=-3)
   )
   &AddVelocity(Velocity=$Output.RotateAroundPoint.LocalForwardVector, Velocity Speed Scale=-10)
  ]
  particleUpdateScript [
   &ParticleState()
   &ScaleSpriteSize(Uniform Curve Sprite Scale=DataInterface() {propertyValues {Curve {keys [
        {arriveTangent=1.2402, interpMode=RCIM_Cubic, leaveTangent=1.2402, tangentMode=RCTM_User, time=0, value=0.55}
        {tangentMode=RCTM_User, time=1, value=1.375}
       ]}}})
   &ScaleColor(
    Scale Alpha=&FloatFromCurve(
     FloatCurve=DataInterface() {propertyValues {Curve {keys [
         {arriveTangent=3.6802, interpMode=RCIM_Cubic, leaveTangent=3.6802, tangentMode=RCTM_User, time=0, value=0}
         {interpMode=RCIM_Cubic, tangentMode=RCTM_User, time=0.5, value=1}
         {tangentMode=RCTM_User, time=1, value=0}
        ]}}}
     Scale Curve=&FNTimeOfDayScalar(Night Scale=0.3, Morning Scale=0.25, Day Scale=0.25, Evening Scale=0.25)
    )
    Scale RGB=&VectorFromFloat(Value=&FNTimeOfDayScalar(Night Scale=0.75, Morning Scale=0.9, Evening Scale=0.9))
    ScaleA=true
    ScaleRGB=true
   )
   &SolveForcesAndVelocity(Clamp Velocity=false, Limit Acceleration=false)
  ]
  rendererClasses [&NiagaraSpriteRendererProperties]
  renderers [{rendererClass=&NiagaraSpriteRendererProperties}]
 }
 Earth_Spawn_Trail {
  emitterUpdateScript [
   &EmitterState(Life Cycle Mode=::NewEnumerator1, Loop Behavior=::NewEnumerator1, Loop Duration=1.5, UseLoopDelay=false)
   &SpawnRate(SpawnRate=120, Use Spawn Probability=false)
  ]
  particleSpawnScript [
   &InitializeParticle(
    Color=&MakeLinearColorFromVectorAndFloat(
     "Vector (RGB)"=&MakeVectorFromLinearColorRGB(Color=<rgba 0.3 0.23 0.18 3>)
     "Float (Alpha)"=&FloatFromCurve(
      FloatCurve=DataInterface() {propertyValues {Curve {keys [
          {arriveTangent=9.3399, interpMode=RCIM_Cubic, leaveTangent=9.3399, tangentMode=RCTM_User, time=0, value=0}
          {interpMode=RCIM_Cubic, tangentMode=RCTM_User, time=0.3, value=1}
          {tangentMode=RCTM_User, time=0.8, value=0}
         ]}}}
      CurveIndex=$Emitter.NormalizedLoopAge
     )
    )
    Lifetime=0.5
    Ribbon Width=&FloatFromCurve(
     FloatCurve=DataInterface() {propertyValues {Curve {keys [
         {arriveTangent=4.5805, interpMode=RCIM_Cubic, leaveTangent=4.5805, tangentMode=RCTM_User, time=0, value=0}
         {interpMode=RCIM_Cubic, tangentMode=RCTM_User, time=0.5, value=1}
        ]}}}
     CurveIndex=$Emitter.NormalizedLoopAge
     Scale Curve=35
    )
    Ribbon Width Mode=::NewEnumerator1
    Sprite Size Mode=::NewEnumerator3
    Uniform Sprite Size=40
    UsePositionOffset=false
    Write Material Random=true
   ) {module=InitializeParticle001}
   &DynamicMaterialParameters(Module.Index0Param0Write=true, Module.Index0Param1Write=true, Module.Index0Param2Write=true, Module.Index0Param3Write=true)
   &RotateAroundPoint(
    Radius:NiagaraFloat=&FloatFromCurve(
     FloatCurve=DataInterface() {propertyValues {Curve {keys [{interpMode=RCIM_Cubic, tangentMode=RCTM_User, time=0, value=0.875}, {interpMode=RCIM_Cubic, tangentMode=RCTM_User, time=1, value=1.125}]}}}
     CurveIndex=$Emitter.NormalizedLoopAge
     Scale Curve=60
    )
    Rotation Center:NiagaraPosition=$System.SampleSkeletalMesh.SampledPosition
    Rotation Phase:NiagaraFloat=&Multiply_Float(
     A=$Emitter.NormalizedLoopAge
     B=&FloatFromCurve(
      FloatCurve=DataInterface() {propertyValues {Curve {keys [{interpMode=RCIM_Cubic, tangentMode=RCTM_User, time=0, value=1}, {interpMode=RCIM_Cubic, tangentMode=RCTM_User, time=1, value=0}]}}}
      Scale Curve=-3
     )
    )
   )
   /SetVariables(Particles.bool:NiagaraBool=&RandomBool()) {isSetParameters=true}
  ]
  particleUpdateScript [
   &ParticleState()
   &ScaleColor(
    Scale Alpha=&FloatFromCurve(
     FloatCurve=DataInterface() {propertyValues {Curve {keys [
         {arriveTangent=3.6802, interpMode=RCIM_Cubic, leaveTangent=3.6802, tangentMode=RCTM_User, time=0, value=0}
         {interpMode=RCIM_Cubic, tangentMode=RCTM_User, time=0.5, value=1}
         {tangentMode=RCTM_User, time=1, value=0}
        ]}}}
     Scale Curve=&FNTimeOfDayScalar(Night Scale=0.7, Morning Scale=0.6, Day Scale=0.6, Evening Scale=0.6)
    )
    Scale RGB=&VectorFromFloat(Value=&FNTimeOfDayScalar(Night Scale=0.8, Morning Scale=0.9, Evening Scale=0.9))
    ScaleA=true
    ScaleRGB=true
   )
   &ScaleRibbonWidth(Ribbon Width Scale=&FloatFromCurve(FloatCurve=DataInterface() {propertyValues {Curve {keys [
         {arriveTangent=1.3079, interpMode=RCIM_Cubic, leaveTangent=1.3079, tangentMode=RCTM_User, time=0, value=0.725}
         {arriveTangent=0.0224, interpMode=RCIM_Cubic, leaveTangent=0.0224, tangentMode=RCTM_User, time=1, value=1.3}
        ]}}}))
  ]
  rendererClasses [&NiagaraRibbonRendererProperties]
  renderers [
   {rendererClass=&NiagaraRibbonRendererProperties}
   {rendererClass=&NiagaraRibbonRendererProperties, rendererIndex=1}
  ]
 }
 Rock_Explode {
  emitterSpawnScript [
   /Game/Effects/Niagara/MasterMaterials/ToonShader/Template/NM_ToonShaderParameters(HighLight_Size=0.1636)
   /Game/Effects/Niagara/MasterMaterials/ToonShader/Template/NM_OutlineParameters(OutlineWidth:NiagaraFloat=&FN_PickFloatDependOnPlatform(TargetPlatformActive=1, TargetPlatformNoActive=0.3))
  ]
  emitterUpdateScript [&EmitterState(Life Cycle Mode=::NewEnumerator1, Loop Behavior=::NewEnumerator1, Loop Duration=6, UseLoopDelay=false)]
  particleSpawnScript [
   &InitializeParticle(
    Color Maximum=<rgba 0.25 0.1375 0.1 1>
    Color Minimum=<rgba 0.5 0.375 0.25 1>
    Color Mode=::NewEnumerator2
    Lifetime Max=0.5
    Lifetime Min=0.25
    Lifetime Mode=::NewEnumerator1
    Mass Mode=::NewEnumerator1
    Mesh Scale Mode=::NewEnumerator1
    Mesh Uniform Scale Max=0.9
    Mesh Uniform Scale Min=0.4
    UsePositionOffset=false
    Write Material Random=true
   )
   /SetVariables(Particles.MeshRandom:NiagaraInt32=&RandomRangeInt(Minimum=0, Maximum=2)) {isSetParameters=true, module=SetVariables_#2}
   &InitialMeshOrientation(Use Rotation Vector=false)
  ]
  particleUpdateScript [
   &ParticleState()
   &ScaleMeshSize(Scale Factor=&VectorFromFloat(Value=&FloatFromCurve(FloatCurve=DataInterface() {propertyValues {Curve {keys [
          {tangentMode=RCTM_User, time=0, value=1}
          {arriveTangent=-0.0405, interpMode=RCIM_Cubic, leaveTangent=-0.0405, tangentMode=RCTM_User, time=0.5, value=1}
          {arriveTangent=-4.2637, interpMode=RCIM_Cubic, leaveTangent=-4.2637, tangentMode=RCTM_User, time=1, value=0}
         ]}}})))
   &ScaleColor(
    Scale RGB=&FN_Select_Vector_from_Time_Of_Day(
     Morning=&MakeVectorFromLinearColorRGB(Color=<rgba 1 1 1 1>)
     Day=&MakeVectorFromLinearColorRGB(Color=&MakeCustomLinearColorFromBool(
       Bool=&Boolean_OrOperation(Bool1=$User.bIsFrontEnd, Bool2=$User.bIsFrontEndPreview)
       True Color=<rgba 1 1 1 1>
       False Color=<rgba 1 1 1 1>
      ))
     Evening=&MakeVectorFromLinearColorRGB(Color=<rgba 1 1 1 1>)
     Night=&MakeVectorFromLinearColorRGB(Color=<rgba 0.81 0.8453 0.9 1>)
    )
    ScaleA=true
    ScaleRGB=true
   )
   &GravityForce()
   &Drag(
    Drag=&FloatFromCurve(
     FloatCurve=DataInterface() {propertyValues {Curve {keys [{interpMode=RCIM_Cubic, tangentMode=RCTM_User, time=0, value=0.5}, {interpMode=RCIM_Cubic, tangentMode=RCTM_User, time=1, value=1}]}}}
     Scale Curve=5
    )
    Use Linear Drag=true
    Use Rotational Drag=false
   )
   &SolveForcesAndVelocity(Clamp Velocity=false, Limit Acceleration=false)
   &GenerateLocationEvent(Event Type:ENiagara_LocEventType=::NewEnumerator2, Use Event Probability=false, UseEventDelay=false)
  ]
  rendererClasses [&NiagaraMeshRendererProperties]
  renderers [
   {rendererClass=&NiagaraMeshRendererProperties, rendererIndex=0}
   {rendererClass=&NiagaraMeshRendererProperties, rendererIndex=1}
  ]
 }
 Rock_Flare {
  emitterUpdateScript [
   &EmitterState(Life Cycle Mode=::NewEnumerator1, Loop Behavior=::NewEnumerator1, Loop Duration=6, UseLoopDelay=false)
   &SpawnBurst_Instantaneous(Spawn Count=7, Spawn Time=0.4, Use Spawn Probability=false)
  ]
  particleSpawnScript [
   &InitializeParticle(
    Color=<rgba 1.5 0.9431 0.6 0.51>
    Lifetime=3
    Sprite Rotation Mode=::NewEnumerator2
    Sprite Size Mode=::NewEnumerator3
    Uniform Sprite Size=40
    UsePositionOffset=false
    Write Material Random=true
   )
   &DynamicMaterialParameters(
    Index 0 Param 1=&RandomRangeFloat(Minimum=0.4, Maximum=0.6)
    Index 0 Param 2=1
    Index 0 Param 3=1
    Index 0 Param 4=1
    Module.Param0WriteEnabled=true
    Module.Param1WriteEnabled=true
    Module.Param2WriteEnabled=true
    Module.Param3WriteEnabled=true
   )
   &CameraOffset()
   &FN_SetSpawnGroup(Number of Spawn Groups=7)
  ]
  particleUpdateScript [
   &ParticleState()
   &self:ScratchModule_07(
    New Particle Attribute Reader=DataInterface() {propertyValues {EmitterBinding {bindingMode=Other, emitterName=Rocks}}}
    Particle Index:NiagaraInt32=&SpawnGroup()
   ) {module=ScratchModule_07}
   &ScaleColor(
    Scale Alpha=&Multiply_Float(
     A=&FNTimeOfDayScalar(Night Scale=0.5, Morning Scale=0.3, Evening Scale=0.4)
     B=&FloatFromCurve(FloatCurve=DataInterface() {propertyValues {Curve {keys [<at 0 0>,<at 0.2 1>,<at 0.8 1>,<at 1 0>]}}})
    )
    Scale RGB=&Multiply_VectorByFloat(Vector=<1 1 1>, Float=&FN_FrontEndFloatModifier(Gameplay=1, Front End=1.2))
    ScaleA=true
    ScaleRGB=true
   )
   &ScaleSpriteSize(Scale Factor=&Vector2DFromFloat(Value=&FloatFromCurve(FloatCurve=DataInterface() {propertyValues {Curve {keys [<at 0 1.5 RCIM_Cubic>,<at 1 0>]}}})))
  ]
  rendererClasses [&NiagaraSpriteRendererProperties]
  renderers [{rendererClass=&NiagaraSpriteRendererProperties}]
 }
 Rock_Flare001 {
  emitterUpdateScript [
   &EmitterState(Life Cycle Mode=::NewEnumerator1, Loop Behavior=::NewEnumerator1, Loop Duration=6, UseLoopDelay=false)
   &SpawnBurst_Instantaneous(Spawn Count=7, Spawn Time=2.1, Use Spawn Probability=false)
  ]
  particleSpawnScript [
   &InitializeParticle(
    Color=<rgba 2 1.2575 0.8 0.51>
    Lifetime=2
    Sprite Rotation Mode=::NewEnumerator2
    Sprite Size Mode=::NewEnumerator3
    Uniform Sprite Size=40
    UsePositionOffset=false
    Write Material Random=true
   )
   &DynamicMaterialParameters(
    Index 0 Param 1=&RandomRangeFloat(Minimum=0.4, Maximum=0.6)
    Index 0 Param 2=1
    Index 0 Param 3=1
    Index 0 Param 4=1
    Module.Param0WriteEnabled=true
    Module.Param1WriteEnabled=true
    Module.Param2WriteEnabled=true
    Module.Param3WriteEnabled=true
   )
   &CameraOffset()
   &FN_SetSpawnGroup(Number of Spawn Groups=7)
  ]
  particleUpdateScript [
   &ParticleState()
   &self:ScratchModule_07(
    New Particle Attribute Reader=DataInterface() {propertyValues {EmitterBinding {bindingMode=Other, emitterName=Rocks}}}
    Particle Index:NiagaraInt32=&SpawnGroup()
   ) {module=ScratchModule_07}
   &ScaleColor(
    Scale Alpha=&Multiply_Float(
     A=&FNTimeOfDayScalar(Night Scale=0.5, Morning Scale=0.6, Evening Scale=0.7)
     B=&FloatFromCurve(FloatCurve=DataInterface() {propertyValues {Curve {keys [<at 0 0>,<at 0.2 1>,<at 0.8 1>,<at 1 0>]}}})
    )
    Scale RGB=&Multiply_VectorByFloat(
     Vector=&VectorFromFloat(Value=&FloatFromCurve(
       FloatCurve=DataInterface() {propertyValues {Curve {keys [
           {arriveTangent=0.5119, interpMode=RCIM_Cubic, leaveTangent=0.5119, tangentMode=RCTM_User, time=0, value=0}
           {arriveTangent=2.8825, interpMode=RCIM_Cubic, leaveTangent=2.8825, tangentMode=RCTM_User, time=0.5, value=0.6}
           {arriveTangent=0.0224, interpMode=RCIM_Cubic, leaveTangent=0.0224, tangentMode=RCTM_User, time=1, value=1.5}
          ]}}}
       Scale Curve=&FNTimeOfDayScalar(Night Scale=4, Morning Scale=4.5, Day Scale=5, Evening Scale=4.5)
      ))
     Float=&FN_FrontEndFloatModifier(Gameplay=1, Front End=1.2)
    )
    ScaleA=true
    ScaleRGB=true
   )
   &ScaleSpriteSize(Scale Factor=&Vector2DFromFloat(Value=&Multiply_Float(A=&MakeFloatFromVector(Vector=$Output.ScratchModule_07.Scale, Channel=::NewEnumerator0), B=1)))
  ]
  rendererClasses [&NiagaraSpriteRendererProperties]
  renderers [{rendererClass=&NiagaraSpriteRendererProperties}]
 }
 Rocks {
  emitterSpawnScript [
   &NM_ToonShaderParameters()
   &NM_OutlineParameters(OutlineWidth:NiagaraFloat=&FN_PickFloatDependOnPlatform(TargetPlatformActive=1, TargetPlatformNoActive=0.5))
   /SetVariables(Emitter.Switch Outline Thickness=1) {isSetParameters=true}
  ]
  emitterUpdateScript [
   &EmitterState(Life Cycle Mode=::NewEnumerator1, Loop Behavior=::NewEnumerator1, UseLoopDelay=false)
   &SpawnBurst_Instantaneous(Spawn Count=7, Spawn Time=0.4, Use Spawn Probability=false, UseLoopCountLimit=false)
  ]
  particleSpawnScript [
   &InitializeParticle(
    Color Maximum=<rgba 0.25 0.1562 0.125 1>
    Color Minimum=<rgba 0.5 0.35 0.2 1>
    Color Mode=::NewEnumerator2
    Lifetime=3.15
    Mass Mode=::NewEnumerator1
    Mesh Scale Mode=::NewEnumerator1
    Mesh Uniform Scale Max=1.5
    UsePositionOffset=false
    Write Material Random=true
   )
   &FN_SetSpawnGroup(Number of Spawn Groups=7)
   /SetVariables(Particles.MeshRandom:NiagaraInt32=&RandomRangeInt(Minimum=0, Maximum=2)) {isSetParameters=true, module=SetVariables_#2}
   &InitialMeshOrientation(Use Rotation Vector=false)
   &CameraOffset(Camera Offset Amount=10)
  ]
  particleUpdateScript [
   &ParticleState()
   &SkeletalMeshLocation(
    Bone / Socket Index=&Add_Integer(A=&SpawnGroup(), B=4)
    Bone Sampling Mode=::NewEnumerator4
    MaskSpawnGroup=false
    Orientation Sampling=::NewEnumerator1
    Skeletal Mesh=$System.Skeletal Mesh
   )
   &ScaleMeshSize(Scale Factor=&VectorFromFloat(Value=&FloatFromCurve(FloatCurve=DataInterface() {propertyValues {Curve {keys [
          {arriveTangent=4.2428, interpMode=RCIM_Cubic, leaveTangent=4.2428, tangentMode=RCTM_User, time=0, value=0}
          {interpMode=RCIM_Cubic, tangentMode=RCTM_User, time=0.4, value=1}
          {arriveTangent=-0.0188, interpMode=RCIM_Cubic, leaveTangent=-0.0188, tangentMode=RCTM_User, time=0.9, value=1}
         ]}}})))
   &UpdateMeshOrientation(
    Rotation Rate=&FloatFromCurve(
     FloatCurve=DataInterface() {propertyValues {Curve {keys [
         <at 0 0.375 RCIM_Cubic>
         {arriveTangent=0.0037, interpMode=RCIM_Cubic, leaveTangent=0.0037, tangentMode=RCTM_User, time=0.6, value=0.375}
         {arriveTangent=8.4356, interpMode=RCIM_Cubic, leaveTangent=8.4356, tangentMode=RCTM_User, time=1, value=2}
        ]}}}
     Scale Curve=&SelectFloatFromArray(
      Array Sampling Mode=::NewEnumerator0
      Direct Array Index=?
      Float Selection Array=DataInterface() {propertyValues {FloatData [-1,-1.5,1,1.5]}}
      Random Seed=0
     )
    )
    Rotation Vector=&RandomVector(Vector Scale=1)
   )
   &SolveForcesAndVelocity(Clamp Velocity=false, Limit Acceleration=false)
   &GenerateLocationEvent(Event Type:ENiagara_LocEventType=::NewEnumerator2, Use Event Probability=false, UseEventDelay=false)
   &GenerateDeathEvent(Use Event Probability=false)
   &ScaleColor(
    Scale RGB=&FN_Select_Vector_from_Time_Of_Day(
     Morning=&MakeVectorFromLinearColorRGB(Color=<rgba 1 1 1 1>)
     Day=&MakeVectorFromLinearColorRGB(Color=&MakeCustomLinearColorFromBool(
       Bool=&Boolean_OrOperation(Bool1=$User.bIsFrontEnd, Bool2=$User.bIsFrontEndPreview)
       True Color=<rgba 1 1 1 1>
       False Color=<rgba 1 1 1 1>
      ))
     Evening=&MakeVectorFromLinearColorRGB(Color=<rgba 1 1 1 1>)
     Night=&MakeVectorFromLinearColorRGB(Color=<rgba 0.81 0.8453 0.9 1>)
    )
    ScaleA=true
    ScaleRGB=true
   )
  ]
  rendererClasses [&NiagaraMeshRendererProperties]
  renderers [
   {rendererClass=&NiagaraMeshRendererProperties, rendererIndex=0}
   {rendererClass=&NiagaraMeshRendererProperties, rendererIndex=1}
   {rendererClass=&NiagaraMeshRendererProperties, rendererIndex=2}
  ]
 }
}
summary {
 systemName=NS_Emote_PlantStand_Earth
 emitters [
  {emitterName=Rocks, rendererClasses [&NiagaraMeshRendererProperties]}
  {emitterName=Rock_Flare, rendererClasses [&NiagaraSpriteRendererProperties]}
  {emitterName=Rock_Flare001, rendererClasses [&NiagaraSpriteRendererProperties]}
  {emitterName=Rock_Explode, rendererClasses [&NiagaraMeshRendererProperties]}
  {emitterName=DryIceSmoke, rendererClasses [&NiagaraSpriteRendererProperties]}
  {emitterName=Earth_Explode_Sparks_Death_Event, rendererClasses [&NiagaraSpriteRendererProperties]}
  {emitterName=Earth_Spawn_Trail, rendererClasses [&NiagaraRibbonRendererProperties]}
  {emitterName=Earth_Spawn_Smoke, rendererClasses [&NiagaraSpriteRendererProperties]}
 ]
 userVariables [
  User.bIsFrontEnd {type=NiagaraBool}
  User.bIsFrontEndPreview {type=NiagaraBool}
 ]
}
