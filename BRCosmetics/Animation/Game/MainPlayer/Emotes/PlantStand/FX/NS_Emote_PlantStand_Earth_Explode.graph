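# Niagara system for the earth explosion effect in the main player's PlantStand emote.
# A system-level skeletal mesh sample positions all emitters: one-shot flare, dark glow, curl-noise sparks with a second glow sprite, toon smoke and earth airflow meshes.
# A delayed, short spawn-rate emitter adds rotating air airflow meshes that fall under gravity.
# Time-of-day scalars dim color or alpha, and front-end modifiers adjust some scale and brightness values.
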NiagaraSystem /BRCosmetics/Animation/Game/MainPlayer/Emotes/PlantStand/FX/NS_Emote_PlantStand_Earth_Explode
dependencies {
 usedDataInterfaces [
  &NiagaraDataInterfaceCurve
  &NiagaraDataInterfaceVector2DCurve
  &NiagaraDataInterfaceArrayFloat
  &NiagaraDataInterfaceSkeletalMesh
 ]
 usedDynamicInputs [&SimulationPosition]
 usedModules [
  &self:SystemSpawnScript.SystemScriptSource.SystemScriptGraph.NiagaraNodeAssignment_0.SetVariables_#1
  &SystemState
  &SampleSkeletalMesh
 ]
}
emitters {
 Air_Airflow_End {
  emitterUpdateScript [
   &EmitterState(Life Cycle Mode=::NewEnumerator1, Loop Behavior=::NewEnumerator1, Loop Delay=0.6, Loop Duration=0.4, UseLoopDelay=true)
   &SpawnRate(Spawn Probability=0.75, SpawnRate=20, Use Spawn Probability=true)
  ]
  particleSpawnScript [
   &InitializeParticle(
    Color Maximum=<rgba 0.4 0.47 0.5 0.1>
    Color Minimum=<rgba 0.8 0.9367 1 0.15>
    Color Mode=::NewEnumerator2
    Lifetime Max=1
    Lifetime Min=0.8
    Lifetime Mode=::NewEnumerator1
    Mesh Scale=&Multiply_VectorByFloat(
     Vector=&RandomRangeVector(Minimum=<1 1 2>, Maximum=<1 1 3>)
     Float=&RandomRangeFloat(Minimum=0.5, Maximum=1)
    )
    Mesh Scale Mode=::NewEnumerator4
    Position=&TransformPosition(Position=$System.SampleSkeletalMesh.SampledPosition, Source Space=::Local, Destination Space=::Simulation)
    Position Mode=::NewEnumerator1
    UsePositionOffset=false
    Write Material Random=true
   )
   &BoxLocation(Box Offset=<0 0 -110>, Box Size=<0 0 10>, MaskSpawnGroup=false, Module.Override Seed=false, Rotate Box=false, Surface Only=false)
   &InitialMeshOrientation(
    Mesh Axis To Orient=<-1 0 0>
    Mesh Orientation Mode=::NewEnumerator2
    Orientation Coordinate Space=::Local
    Rotation=&Multiply_VectorByFloat(Vector=&RandomRangeVector(Minimum=<-1 -1 -100>, Maximum=<1 1 100>), Float=0.025)
    Use Rotation Vector=true
   )
  ]
  particleUpdateScript [
   &ParticleState()
   &DynamicMaterialParameters(
    Index 0 Param 1=&FloatFromCurve(FloatCurve=DataInterface() {propertyValues {Curve {keys [
         {interpMode=RCIM_Cubic, tangentMode=RCTM_User, time=0, value=1}
         <at 0.7 0.35>
         {interpMode=RCIM_Cubic, tangentMode=RCTM_User, time=1, value=0}
        ]}}})
    Module.Param0WriteEnabled=true
    Module.Param1WriteEnabled=true
    Module.Param2WriteEnabled=true
    Module.Param3WriteEnabled=true
   )
   &ScaleMeshSize(Scale Factor=&Multiply_VectorByFloat(
     Vector=&Multiply_VectorByFloat(Vector=$Engine.Owner.Scale, Float=1)
     Float=&FloatFromCurve(
      FloatCurve=DataInterface() {propertyValues {Curve {keys [
          {interpMode=RCIM_Cubic, tangentMode=RCTM_User, time=0, value=0.2}
          {arriveTangent=1.688, interpMode=RCIM_Cubic, leaveTangent=1.688, tangentMode=RCTM_User, time=1, value=1}
         ]}}}
      Scale Curve=&FN_FrontEndFloatModifier(Gameplay=2.5, Front End=2.5)
     )
    ))
   &UpdateMeshOrientation(
    Rotation Rate=&SelectFloatFromArray(
     Array Sampling Mode=::NewEnumerator0
     Direct Array Index=?
     Float Selection Array=DataInterface() {propertyValues {FloatData [-0.6,-0.3,0.25]}}
     Random Seed=0
    )
    Rotation Vector=<0 0 1>
   )
   &Drag(Drag=10, Use Linear Drag=true, Use Rotational Drag=true)
   &GravityForce(Gravity=<0 0 -600>)
   &SolveForcesAndVelocity(Clamp Velocity=false, Limit Acceleration=false)
   &ScaleColor(
    Scale Alpha=&FloatFromCurve(
     FloatCurve=DataInterface() {propertyValues {Curve {keys [
         {arriveTangent=6.476, interpMode=RCIM_Cubic, leaveTangent=6.476, tangentMode=RCTM_User, time=0, value=0}
         {interpMode=RCIM_Cubic, tangentMode=RCTM_User, time=0.3, value=1}
         {tangentMode=RCTM_User, time=1, value=0}
        ]}}}
     Scale Curve=0.8
    )
    Scale RGB=&Multiply_VectorByFloat(
     Vector=&Multiply_VectorByFloat(Vector=<1 1 1>, Float=&FN_FrontEndFloatModifier(Gameplay=1, Front End=1))
     Float=&FNTimeOfDayScalar(Night Scale=0.5, Morning Scale=0.3, Evening Scale=0.3)
    )
    ScaleA=true
    ScaleRGB=true
   )
  ]
  rendererClasses [&NiagaraMeshRendererProperties]
  renderers [{rendererClass=&NiagaraMeshRendererProperties, rendererIndex=0}]
 }
 Dark_Glow001 {
  emitterUpdateScript [
   &EmitterState(Life Cycle Mode=::NewEnumerator1, Loop Behavior=::NewEnumerator1, Loop Delay=0.1, Loop Duration=6, UseLoopDelay=false)
   &SpawnBurst_Instantaneous(Spawn Count=1, Use Spawn Probability=false)
  ]
  particleSpawnScript [
   &InitializeParticle(
    Color=<rgba 0.03 0.0121 0.0048 1>
    Lifetime=0.4
    Position Offset=$System.SampleSkeletalMesh.SampledPosition
    Sprite Rotation Mode=::NewEnumerator2
    Sprite Size Mode=::NewEnumerator3
    Uniform Sprite Size=600
    UsePositionOffset=true
    Write Material Random=true
   )
   &DynamicMaterialParameters(
    Index 0 Param 1=&RandomRangeFloat(Minimum=0.4, Maximum=0.6)
    Index 0 Param 2=1
    Index 0 Param 3=1
    Index 0 Param 4=1
    Module.Param0WriteEnabled=true
    Module.Param1WriteEnabled=true
    Module.Param2WriteEnabled=true
    Module.Param3WriteEnabled=true
   )
   &CameraOffset(Camera Offset Amount=-10)
  ]
  particleUpdateScript [
   &ParticleState()
   &SolveForcesAndVelocity(Clamp Velocity=false, Limit Acceleration=false)
   &ScaleColor(
    Scale Alpha=&Multiply_Float(
     A=1
     B=&FloatFromCurve(FloatCurve=DataInterface() {propertyValues {Curve {keys [<at 0 0>,<at 0.2 1>,<at 0.8 1>,<at 1 0>]}}})
    )
    Scale RGB=&Multiply_VectorByFloat(Vector=<1 1 1>, Float=&FNTimeOfDayScalar(Night Scale=0.5, Morning Scale=0.3, Evening Scale=0.4))
    ScaleA=true
    ScaleRGB=true
   )
   &ScaleSpriteSize(Scale Factor=&Vector2DFromFloat(Value=&FloatFromCurve(FloatCurve=DataInterface() {propertyValues {Curve {keys [<at 0 1.5 RCIM_Cubic>,<at 1 0>]}}})))
  ]
  rendererClasses [&NiagaraSpriteRendererProperties]
  renderers [{rendererClass=&NiagaraSpriteRendererProperties}]
 }
 Earth_Airflow {
  emitterUpdateScript [
   &EmitterState(Life Cycle Mode=::NewEnumerator1, Loop Behavior=::NewEnumerator1, Loop Duration=6, UseLoopDelay=true)
   &SpawnBurst_Instantaneous(Spawn Count=4, Use Spawn Probability=false, UseLoopCountLimit=false)
  ]
  particleSpawnScript [
   &InitializeParticle(
    Color Maximum=<rgba 0.4 0.34 0.28 0.1>
    Color Minimum=<rgba 0.8 0.5867 0.48 0.2>
    Color Mode=::NewEnumerator2
    Lifetime Max=1
    Lifetime Min=0.6
    Lifetime Mode=::NewEnumerator1
    Mesh Scale=&Multiply_VectorByFloat(
     Vector=&RandomRangeVector(Minimum=<1 1 1>, Maximum=<1 1 1.5>)
     Float=&RandomRangeFloat(Minimum=1.2, Maximum=2)
    )
    Mesh Scale Mode=::NewEnumerator4
    Position Offset=&AddVectorToPosition(Position=$System.SampleSkeletalMesh.SampledPosition, Vector=<0 0 -20>)
    UsePositionOffset=true
    Write Material Random=true
   )
   &InitialMeshOrientation(Use Rotation Vector=true)
  ]
  particleUpdateScript [
   &ParticleState()
   &DynamicMaterialParameters(
    Index 0 Param 1=&FloatFromCurve(FloatCurve=DataInterface() {propertyValues {Curve {keys [
         {interpMode=RCIM_Cubic, tangentMode=RCTM_User, time=0, value=1}
         <at 0.7 0.35>
         {interpMode=RCIM_Cubic, tangentMode=RCTM_User, time=1, value=0}
        ]}}})
    Module.Param0WriteEnabled=true
    Module.Param1WriteEnabled=true
    Module.Param2WriteEnabled=true
    Module.Param3WriteEnabled=true
   )
   &ScaleMeshSize(Scale Factor=&Multiply_VectorByFloat(
     Vector=&Multiply_VectorByFloat(Vector=$Engine.Owner.Scale, Float=1)
     Float=&FloatFromCurve(FloatCurve=DataInterface() {propertyValues {Curve {keys [
          {arriveTangent=2.8996, interpMode=RCIM_Cubic, leaveTangent=2.8996, tangentMode=RCTM_User, time=0, value=0}
          {arriveTangent=0.0201, interpMode=RCIM_Cubic, leaveTangent=0.0201, tangentMode=RCTM_User, time=1, value=1.2}
         ]}}})
    ))
   &UpdateMeshOrientation(Rotation Vector=<0 0 1>)
   &ScaleColor(
    Scale Alpha=&FloatFromCurve(
     FloatCurve=DataInterface() {propertyValues {Curve {keys [
         {arriveTangent=6.7494, interpMode=RCIM_Cubic, leaveTangent=6.7494, tangentMode=RCTM_User, time=0, value=0}
         {interpMode=RCIM_Cubic, tangentMode=RCTM_User, time=0.3, value=1}
         {tangentMode=RCTM_User, time=1, value=0}
        ]}}}
     Scale Curve=0.75
    )
    Scale RGB=&Multiply_VectorByFloat(
     Vector=&Multiply_VectorByFloat(Vector=<1 1 1>, Float=&FN_FrontEndFloatModifier(Gameplay=1, Front End=1))
     Float=&FNTimeOfDayScalar(Night Scale=0.5, Morning Scale=0.3, Day Scale=0.75, Evening Scale=0.3)
    )
    ScaleA=true
    ScaleRGB=true
   )
  ]
  rendererClasses [&NiagaraMeshRendererProperties]
  renderers [{rendererClass=&NiagaraMeshRendererProperties, rendererIndex=0}]
 }
 Earth_Explode_Sparks {
  emitterUpdateScript [
   &EmitterState(Life Cycle Mode=::NewEnumerator1, Loop Behavior=::NewEnumerator1, Loop Delay=0.1, Loop Duration=6, UseLoopDelay=false)
   &SpawnBurst_Instantaneous(Spawn Count=30, Use Spawn Probability=false, UseLoopCountLimit=false)
  ]
  particleSpawnScript [
   &InitializeParticle(
    Color=<rgba 1 0.3293 0.1562 1>
    Lifetime=&RandomRangeFloat(Minimum=0.3, Maximum=0.8)
    Position Offset=$System.SampleSkeletalMesh.SampledPosition
    Sprite Rotation Mode=::NewEnumerator0
    Sprite Size Max=<2 5>
    Sprite Size Min=<1 3>
    Sprite Size Mode=::NewEnumerator2
    UsePositionOffset=true
    Write Material Random=true
   )
   /SetVariables(Particles.RandomCore:NiagaraFloat=&FloatFromCurve(
     FloatCurve=DataInterface() {propertyValues {Curve {keys [
         {arriveTangent=0.0559, interpMode=RCIM_Cubic, leaveTangent=0.0559, tangentMode=RCTM_User, time=0, value=0.2}
         {arriveTangent=0.1895, interpMode=RCIM_Cubic, leaveTangent=0.1895, tangentMode=RCTM_User, time=0.9, value=0.25}
         {arriveTangent=1.1808, interpMode=RCIM_Cubic, leaveTangent=1.1808, tangentMode=RCTM_User, time=1, value=0.3}
        ]}}}
     CurveIndex=&RandomRangeFloat(Minimum=0, Maximum=1)
     Scale Curve=2
    )) {isSetParameters=true}
   &ShapeLocation(MaskSpawnGroup=false, Override Seed=false, Sphere Radius=30, UseOwnerScale=true)
   &AddVelocity(
    Module.Use Falloff Distance=false
    Override Seed=false
    Velocity Mode=::NewEnumerator1
    Velocity Origin=$Particles.Position
    Velocity Speed=&RandomRangeFloat(Minimum=100, Maximum=200)
   )
  ]
  particleUpdateScript [
   &ParticleState()
   &CurlNoiseForce(Noise Strength=&Multiply_Float(A=2000, B=1), Pan Noise=true, Pan Noise Field=<0 0 1>)
   &ScaleSpriteSize(Scale Factor=&Vector2DFromFloat(Value=&FloatFromCurve(
      FloatCurve=DataInterface() {propertyValues {Curve {keys [
          <at 0 0 RCIM_Cubic>
          <at 0.2 1.2 RCIM_Cubic>
          {arriveTangent=-2.7132, interpMode=RCIM_Cubic, leaveTangent=-2.7132, tangentMode=RCTM_User, time=1, value=0}
         ]}}}
      Scale Curve=0.5
     )))
   /SetVariables(Particles.Glow03_Size=&Multiply_Vector2D(A=$Particles.SpriteSize, B=<4 3>)) {isSetParameters=true}
   &ScaleColor(
    Scale Alpha=&Multiply_Float(A=1, B=1)
    Scale RGB=&Multiply_VectorByFloat(Vector=<2 2 2>, Float=&FNTimeOfDayScalar(Night Scale=1, Morning Scale=0.7, Evening Scale=0.7))
    ScaleA=true
    ScaleRGB=true
   )
   /SetVariables(Particles.Glow03_Color=&MultiplyLinearColor(
     Linear Color=$Particles.Color
     Scale Factor=&MakeFloat4FromVectorAndFloat("Vector (RGB)"=<1.5 1.5 1.5>, "Float (Alpha)"=0.1)
    )) {isSetParameters=true}
   &Drag(Drag=2, Use Linear Drag=true, Use Rotational Drag=false)
   &SolveForcesAndVelocity(Clamp Velocity=false, Limit Acceleration=false)
  ]
  rendererClasses [&NiagaraSpriteRendererProperties]
  renderers [
   {rendererClass=&NiagaraSpriteRendererProperties}
   {rendererClass=&NiagaraSpriteRendererProperties, rendererIndex=1}
  ]
 }
 Flare001 {
  emitterUpdateScript [
   &EmitterState(Life Cycle Mode=::NewEnumerator1, Loop Behavior=::NewEnumerator1, Loop Duration=6, UseLoopDelay=false)
   &SpawnBurst_Instantaneous(Spawn Count=1, Use Spawn Probability=false)
  ]
  particleSpawnScript [
   &InitializeParticle(
    Color=<rgba 2 1.4167 1 1>
    Lifetime=0.4
    Position Offset=$System.SampleSkeletalMesh.SampledPosition
    Sprite Rotation Mode=::NewEnumerator2
    Sprite Size Mode=::NewEnumerator3
    Uniform Sprite Size=500
    UsePositionOffset=true
    Write Material Random=true
   )
   &DynamicMaterialParameters(
    Index 0 Param 1=&RandomRangeFloat(Minimum=0.4, Maximum=0.6)
    Index 0 Param 2=1
    Index 0 Param 3=1
    Index 0 Param 4=1
    Module.Param0WriteEnabled=true
    Module.Param1WriteEnabled=true
    Module.Param2WriteEnabled=true
    Module.Param3WriteEnabled=true
   )
   &CameraOffset()
  ]
  particleUpdateScript [
   &ParticleState()
   &SolveForcesAndVelocity(Clamp Velocity=false, Limit Acceleration=false)
   &ScaleColor(
    Scale Alpha=&Multiply_Float(
     A=&FNTimeOfDayScalar(Night Scale=0.5, Morning Scale=0.6, Evening Scale=0.6)
     B=&FloatFromCurve(FloatCurve=DataInterface() {propertyValues {Curve {keys [<at 0 0>,<at 0.2 1>,<at 0.8 1>,<at 1 0>]}}})
    )
    Scale RGB=&Multiply_VectorByFloat(Vector=<1 1 1>, Float=&FN_FrontEndFloatModifier(Gameplay=1, Front End=1.2))
    ScaleA=true
    ScaleRGB=true
   )
   &ScaleSpriteSize(Scale Factor=&Vector2DFromFloat(Value=&FloatFromCurve(FloatCurve=DataInterface() {propertyValues {Curve {keys [<at 0 1.5 RCIM_Cubic>,<at 1 0>]}}})))
  ]
  rendererClasses [&NiagaraSpriteRendererProperties]
  renderers [{rendererClass=&NiagaraSpriteRendererProperties}]
 }
 Toon_Smoke {
  emitterUpdateScript [
   &EmitterState(Life Cycle Mode=::NewEnumerator1, Loop Behavior=::NewEnumerator1, Loop Duration=7, UseLoopDelay=false)
   &SpawnBurst_Instantaneous(Spawn Count=4, Spawn Time=0.05, Use Spawn Probability=false)
  ]
  particleSpawnScript [
   /SetVariables(Particles.Random=&RandomRangeFloat(Minimum=0, Maximum=1)) {isSetParameters=true}
   &InitializeParticle(
    Color=<rgba 0.8 0.6133 0.48 0.4>
    Lifetime=&Lerp_Float(A=0.85, B=1, Alpha=$Particles.Random)
    Mass Mode=::NewEnumerator1
    Position Offset=$System.SampleSkeletalMesh.SampledPosition
    Sprite Rotation Angle=&Add_Float(
     A=&ReturnNormalizedExecIndex(Normalized Index Includes One=false)
     B=&RandomRangeFloat(Minimum=-0.1, Maximum=0.1)
    )
    Sprite Rotation Mode=::NewEnumerator1
    Sprite Size=<250 275>
    Sprite Size Mode=::NewEnumerator4
    UsePositionOffset=true
    Write Material Random=true
   ) {module=InitializeParticle002}
   &CameraOffset(Camera Offset Amount=-5)
   &DynamicMaterialParameters(
    Index 0 Param 2=&UniformAOrBFloat(A=0)
    Module.Param0WriteEnabled=true
    Module.Param1WriteEnabled=true
    Module.Param2WriteEnabled=true
    Module.Param3WriteEnabled=true
   ) {module=DynamicMaterialParameters001}
  ]
  particleUpdateScript [
   &ParticleState()
   &ScaleSpriteSize(
    Non-Uniform Curve Sprite Scale=DataInterface() {propertyValues {
      XCurve {keys [
        {arriveTangent=3.0351, interpMode=RCIM_Cubic, leaveTangent=3.0351, tangentMode=RCTM_User, time=0, value=0.2}
        {interpMode=RCIM_Cubic, tangentMode=RCTM_User, time=0.5, value=1}
       ]}
      YCurve {keys [
        {arriveTangent=3.0351, interpMode=RCIM_Cubic, leaveTangent=3.0351, tangentMode=RCTM_User, time=0, value=0.2}
        {interpMode=RCIM_Cubic, tangentMode=RCTM_User, time=0.5, value=1}
       ]}
     }}
    Scale Sprite Size Mode=::NewEnumerator3
   )
   &SolveForcesAndVelocity(Clamp Velocity=false, Limit Acceleration=false)
   &DynamicMaterialParameters(
    Index 0 Param 1=&FloatFromCurve(FloatCurve=DataInterface() {propertyValues {Curve {keys [
         {arriveTangent=1.3581, interpMode=RCIM_Cubic, leaveTangent=1.3581, tangentMode=RCTM_User, time=0, value=0}
         {interpMode=RCIM_Cubic, tangentMode=RCTM_User, time=1, value=1}
        ]}}})
    Index 0 Param 2=1
    Module.Param0WriteEnabled=true
    Module.Param1WriteEnabled=true
    Module.Param2WriteEnabled=true
    Module.Param3WriteEnabled=true
   )
   &ScaleColor(
    Color Value To Scale=$Particles.Color
    Scale Alpha=&FloatFromCurve(
     FloatCurve=DataInterface() {propertyValues {Curve {keys [
         {arriveTangent=-0.5933, interpMode=RCIM_Cubic, leaveTangent=-0.5933, tangentMode=RCTM_User, time=0.2, value=1}
         {arriveTangent=-1.9134, interpMode=RCIM_Cubic, leaveTangent=-1.9134, tangentMode=RCTM_User, time=1, value=0}
        ]}}}
     Scale Curve=&FNTimeOfDayScalar(Night Scale=0.9, Morning Scale=1, Evening Scale=0.9)
    )
    ScaleA=true
    ScaleRGB=true
   )
  ]
  rendererClasses [&NiagaraSpriteRendererProperties]
  renderers [{rendererClass=&NiagaraSpriteRendererProperties}]
 }
}
summary {
 systemName=NS_Emote_PlantStand_Earth_Explode
 emitters [
  {emitterName=Earth_Airflow, rendererClasses [&NiagaraMeshRendererProperties]}
  {emitterName=Flare001, rendererClasses [&NiagaraSpriteRendererProperties]}
  {emitterName=Dark_Glow001, rendererClasses [&NiagaraSpriteRendererProperties]}
  {emitterName=Earth_Explode_Sparks, rendererClasses [&NiagaraSpriteRendererProperties]}
  {emitterName=Toon_Smoke, rendererClasses [&NiagaraSpriteRendererProperties]}
  {emitterName=Air_Airflow_End, rendererClasses [&NiagaraMeshRendererProperties]}
 ]
 userVariables [
  User.bIsFrontEnd {type=NiagaraBool}
  User.bIsFrontEndPreview {type=NiagaraBool}
 ]
}
