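# Niagara system for the PlantStand emote that produces a short burst of fire on a skeletal mesh.
# Ribbon flames, curl-noise sparks and flame flipbook sprites spawn from mesh sampling.
# Sparks inherit mesh velocity, while flame sprites drift along the sampled velocity.
# A fading light tracks a sampled bone or socket, with front-end modifiers dimming light and flame sprites and time-of-day functions adjusting spark brightness.
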NiagaraSystem /BRCosmetics/Animation/Game/MainPlayer/Emotes/PlantStand/FX/NS_Emote_PlantStand_Fire
dependencies {
 usedDataInterfaces [
  &NiagaraDataInterfaceCurve
  &NiagaraDataInterfaceArrayFloat
  &NiagaraDataInterfaceVector2DCurve
  &NiagaraDataInterfaceSkeletalMesh
 ]
 usedDynamicInputs [&MakeVector2D, &SelectFloatFromArray]
 usedModules [
  &self:SystemSpawnScript.SystemScriptSource.SystemScriptGraph.NiagaraNodeAssignment_0.SetVariables_#1
  &SystemState
  &SampleSkeletalMesh
 ]
}
emitters {
 Fire {
  emitterUpdateScript [
   &EmitterState(Life Cycle Mode=::NewEnumerator1, Loop Behavior=::NewEnumerator1, Loop Duration=0.6, UseLoopDelay=false)
   &SpawnRate(SpawnRate=60, Use Spawn Probability=false)
  ]
  particleSpawnScript [
   &InitializeParticle(
    Color=<rgba 2 0.0667 0 0.5>
    Lifetime=0.5
    Ribbon Width=&FloatFromCurve(
     FloatCurve=DataInterface() {propertyValues {Curve {keys [
         {arriveTangent=12.1055, interpMode=RCIM_Cubic, leaveTangent=12.1055, tangentMode=RCTM_User, time=0.1, value=0}
         {arriveTangent=-0.0457, interpMode=RCIM_Cubic, leaveTangent=-0.0457, tangentMode=RCTM_User, time=0.2, value=1}
         {arriveTangent=0.0172, interpMode=RCIM_Cubic, leaveTangent=-5.0275, tangentMode=RCTM_Break, time=0.5, value=1}
         {arriveTangent=-15.1596, interpMode=RCIM_Cubic, leaveTangent=-15.1596, tangentMode=RCTM_User, time=0.6, value=0}
        ]}}}
     CurveIndex=$Emitter.Age
     Scale Curve=40
    )
    Ribbon Width Mode=::NewEnumerator1
    UsePositionOffset=false
    Write Material Random=true
   ) {module=InitializeParticle001}
   &SkeletalMeshLocation(
    Bone Sampling Mode=::NewEnumerator4
    Interpolate Mesh Sampling=false
    MaskSpawnGroup=false
    Orientation Sampling=::NewEnumerator1
    Skeletal Mesh=$System.Skeletal Mesh
   )
   &DynamicMaterialParameters(
    Index 0 Param 3=15
    Index 0 Param 4=0.3
    Module.Index0Param0Write=true
    Module.Index0Param1Write=true
    Module.Index0Param2Write=true
    Module.Index0Param3Write=true
   )
   /SetVariables(Particles.Linear Color=&FN_Select_Color_from_Time_Of_Day(Morning=<rgba 1 1 1 1>, Day=<rgba 1 1 1 1>, Evening=<rgba 1 1 1 1>, Night=<rgba 0.75 0.75 0.75 1>)) {isSetParameters=true}
  ]
  particleUpdateScript [
   &ParticleState()
   &ScaleColor(ScaleA=true, ScaleRGB=true)
   &ScaleRibbonWidth(Ribbon Width Scale=&FloatFromCurve(FloatCurve=DataInterface() {propertyValues {Curve {keys [
         {arriveTangent=6.6969, interpMode=RCIM_Cubic, leaveTangent=6.6969, tangentMode=RCTM_User, time=0, value=0}
         {arriveTangent=0.0075, interpMode=RCIM_Cubic, leaveTangent=0.0075, tangentMode=RCTM_User, time=0.4, value=1.1}
         {arriveTangent=-1.8607, interpMode=RCIM_Cubic, leaveTangent=-1.8607, tangentMode=RCTM_User, time=1, value=0}
        ]}}}))
   &Drag(Drag=5, Use Linear Drag=true, Use Rotational Drag=true)
   &CurlNoiseForce(Noise Frequency=10, Noise Strength=5, Pan Noise=true, Pan Noise Field=<0 0 0.1>)
   &SolveForcesAndVelocity(Clamp Velocity=false, Limit Acceleration=false)
  ]
  rendererClasses [&NiagaraRibbonRendererProperties]
  renderers [
   {rendererClass=&NiagaraRibbonRendererProperties}
   {rendererClass=&NiagaraRibbonRendererProperties, rendererIndex=1}
  ]
 }
 Fire_Sparks {
  emitterUpdateScript [
   &EmitterState(Life Cycle Mode=::NewEnumerator1, Loop Behavior=::NewEnumerator1, Loop Duration=0.6, UseLoopDelay=false)
   &SpawnRate(Spawn Probability=0.8, SpawnRate=59, Use Spawn Probability=true)
   &SpawnRate(Spawn Probability=0.8, SpawnRate=120, Use Spawn Probability=true) {module=SpawnRate001}
  ]
  particleSpawnScript [
   &InitializeParticle(
    Color Maximum=<rgba 1 0.0833 0 1>
    Color Minimum=<rgba 1.5 0.2965 0 1>
    Color Mode=::NewEnumerator2
    Lifetime=&RandomRangeFloat(Minimum=0.5, Maximum=1)
    Sprite Size Max=<2 8>
    Sprite Size Min=<1 6>
    Sprite Size Mode=::NewEnumerator2
    UsePositionOffset=false
    Write Material Random=true
   )
   &SkeletalMeshLocation(
    Bone Sampling Mode=::NewEnumerator4
    "Inherit Velocity (Scale)":NiagaraFloat=&RandomRangeFloat(Minimum=0.5, Maximum=1)
    Inherited Velocity Speed Limit=300
    MaskSpawnGroup=false
    Orientation Sampling=::NewEnumerator1
    Skeletal Mesh=$System.Skeletal Mesh
    Velocity Sampling=::NewEnumerator0
    VelocitySpeedLimit=true
   )
   /SetVariables(Particles.RandomCore:NiagaraFloat=&FloatFromCurve(
     FloatCurve=DataInterface() {propertyValues {Curve {keys [
         {arriveTangent=0.0559, interpMode=RCIM_Cubic, leaveTangent=0.0559, tangentMode=RCTM_User, time=0, value=0.2}
         {arriveTangent=0.1895, interpMode=RCIM_Cubic, leaveTangent=0.1895, tangentMode=RCTM_User, time=0.9, value=0.25}
         {arriveTangent=1.1808, interpMode=RCIM_Cubic, leaveTangent=1.1808, tangentMode=RCTM_User, time=1, value=0.3}
        ]}}}
     CurveIndex=&RandomRangeFloat(Minimum=0, Maximum=1)
     Scale Curve=2
    )) {isSetParameters=true}
   &ShapeLocation(MaskSpawnGroup=false, Override Seed=false, Sphere Radius=5, UseOwnerScale=true)
  ]
  particleUpdateScript [
   &ParticleState()
   &CurlNoiseForce(Noise Strength=300, Pan Noise=true, Pan Noise Field=<0 0 1>)
   &ScaleSpriteSize(Scale Factor=&Vector2DFromFloat(Value=&FloatFromCurve(
      FloatCurve=DataInterface() {propertyValues {Curve {keys [
          {arriveTangent=12.3508, interpMode=RCIM_Cubic, leaveTangent=12.3508, tangentMode=RCTM_User, time=0, value=0}
          <at 0.2 1.2 RCIM_Cubic>
          {arriveTangent=-2.7132, interpMode=RCIM_Cubic, leaveTangent=-2.7132, tangentMode=RCTM_User, time=1, value=0}
         ]}}}
      Scale Curve=0.5
     )))
   /SetVariables(Particles.Glow03_Size=&Multiply_Vector2D(A=$Particles.SpriteSize, B=<4 3>)) {isSetParameters=true}
   &ScaleColor(
    Scale Alpha=&FloatFromCurve(FloatCurve=DataInterface() {propertyValues {Curve {keys [
         {arriveTangent=9.0725, interpMode=RCIM_Cubic, leaveTangent=9.0725, tangentMode=RCTM_User, time=0, value=0}
         {arriveTangent=-2.5249, interpMode=RCIM_Cubic, leaveTangent=-2.5249, tangentMode=RCTM_User, time=0.1408, value=0.9847}
         {arriveTangent=-0.0405, interpMode=RCIM_Cubic, leaveTangent=-0.0405, tangentMode=RCTM_User, time=1, value=0}
        ]}}})
    Scale RGB=&Multiply_VectorByFloat(Vector=<2 2 2>, Float=&FNTimeOfDayScalar(Night Scale=1, Morning Scale=0.7, Evening Scale=0.7))
    ScaleA=true
    ScaleRGB=true
   )
   /SetVariables(Particles.Glow03_Color=&MultiplyLinearColor(
     Linear Color=$Particles.Color
     Scale Factor=&MakeFloat4FromVectorAndFloat("Vector (RGB)"=<1.5 1.5 1.5>, "Float (Alpha)"=0.1)
    )) {isSetParameters=true}
   &Drag(Drag=5, Use Linear Drag=true, Use Rotational Drag=false)
   &SolveForcesAndVelocity(Clamp Velocity=false, Limit Acceleration=false)
  ]
  rendererClasses [&NiagaraSpriteRendererProperties]
  renderers [
   {rendererClass=&NiagaraSpriteRendererProperties}
   {rendererClass=&NiagaraSpriteRendererProperties, rendererIndex=1}
  ]
 }
 FlameSprites {
  emitterUpdateScript [
   &EmitterState(Life Cycle Mode=::NewEnumerator1, Loop Behavior=::NewEnumerator1, Loop Delay=0.1, Loop Duration=0.5, UseLoopDelay=true)
   &SpawnRate(Spawn Probability=0.8, SpawnRate=70, Use Spawn Probability=false)
   &SpawnRate(Spawn Probability=0.85, SpawnRate=45, Use Spawn Probability=true) {module=SpawnRate001}
  ]
  particleSpawnScript [
   &InitializeParticle(
    Color Maximum=<rgba 0.5 0.0417 0 0.3>
    Color Minimum=<rgba 1 0.2 0 0.5>
    Color Mode=::NewEnumerator2
    Lifetime Max=0.8
    Lifetime Min=0.4
    Lifetime Mode=::NewEnumerator1
    Sprite Rotation Mode=::NewEnumerator2
    Sprite Size Mode=::NewEnumerator1
    Sprite UV Mode=::NewEnumerator3
    Uniform Sprite Size Max=55
    Uniform Sprite Size Min=40
    UsePositionOffset=false
    Write Material Random=true
   ) {module=InitializeParticle001}
   &SkeletalMeshLocation(Bone Sampling Mode=::NewEnumerator4, MaskSpawnGroup=false, Orientation Sampling=::NewEnumerator1, Skeletal Mesh=$System.Skeletal Mesh)
   &CameraOffset(Camera Offset Amount=2)
  ]
  particleUpdateScript [
   /SetVariables(Particles.SubImageIndex=&FloatFromCurve(FloatCurve=DataInterface() {propertyValues {Curve {keys [<at 0 0>,<at 1 15>]}}})) {isSetParameters=true}
   &ParticleState()
   /SetVariables(Particles.Position=&AddVectorToPosition(
     Position=$Particles.Position
     Vector=&Multiply_VectorByFloat(
      Vector=&Add_Vector(
       A=&NormalizeVector(
        Vector To Normalize=$Particles.SkeletalMeshLocation.SampledVelocity
        Safe Normalize=true
        Safe Fallback Threshold=0
        Safe Fallback Vector=<0 0 0.001>
       )
       B=&RandomVector(Vector Scale=0.1)
      )
      Float=&FloatFromCurve(
       FloatCurve=DataInterface() {propertyValues {Curve {keys [
           {arriveTangent=2.7136, interpMode=RCIM_Cubic, leaveTangent=2.7136, tangentMode=RCTM_User, time=0, value=0}
           {arriveTangent=0.0224, interpMode=RCIM_Cubic, leaveTangent=0.0224, tangentMode=RCTM_User, time=1, value=1}
          ]}}}
       Scale Curve=&RandomRangeFloat(Minimum=2, Maximum=5)
      )
     )
    )) {isSetParameters=true}
   &ScaleSpriteSize(Scale Factor=&Vector2DFromCurve(
     Vector2Curve=DataInterface() {propertyValues {
       XCurve {keys [
         {arriveTangent=0.7462, interpMode=RCIM_Cubic, leaveTangent=0.7462, tangentMode=RCTM_User, time=0, value=0.6}
         {interpMode=RCIM_Cubic, tangentMode=RCTM_User, time=1, value=1}
        ]}
       YCurve {keys [
         {arriveTangent=0.7462, interpMode=RCIM_Cubic, leaveTangent=0.7462, tangentMode=RCTM_User, time=0, value=0.6}
         {interpMode=RCIM_Cubic, tangentMode=RCTM_User, time=1, value=1}
        ]}
      }}
     ScaleCurve=<1 1>
    ))
   &ScaleColor(
    Scale Alpha=&Multiply_Float(
     A=&FloatFromCurve(
      FloatCurve=DataInterface() {propertyValues {Curve {keys [
          {arriveTangent=9.0725, interpMode=RCIM_Cubic, leaveTangent=9.0725, tangentMode=RCTM_User, time=0, value=0}
          {arriveTangent=-2.5249, interpMode=RCIM_Cubic, leaveTangent=-2.5249, tangentMode=RCTM_User, time=0.1408, value=0.9847}
          {arriveTangent=-0.0405, interpMode=RCIM_Cubic, leaveTangent=-0.0405, tangentMode=RCTM_User, time=1, value=0}
         ]}}}
      Scale Curve=&FN_FrontEndFloatModifier(Gameplay=1, Front End=0.2)
     )
     B=&RandomRangeFloat(Minimum=1, Maximum=2)
    )
    Scale RGB=&VectorFromFloat(Value=1)
    ScaleA=true
    ScaleRGB=true
   ) {module=ScaleColor001}
  ]
  rendererClasses [&NiagaraSpriteRendererProperties]
  renderers [{rendererClass=&NiagaraSpriteRendererProperties}]
 }
 Light_for_Fire {
  emitterUpdateScript [
   &EmitterState(Life Cycle Mode=::NewEnumerator1, Loop Behavior=::NewEnumerator1, Loop Delay=1, Loop Duration=2, UseLoopDelay=false)
   &SpawnBurst_Instantaneous(Spawn Time=0.2, Use Spawn Probability=false, UseLoopCountLimit=false)
  ]
  particleSpawnScript [
   &InitializeParticle(Color=<rgba 1 0.2667 0 0.7>, Sprite Size Mode=::NewEnumerator3, Uniform Sprite Size=100, UsePositionOffset=false, Write Material Random=true)
   &DynamicMaterialParameters(
    Index 0 Param 1=0.2
    Index 0 Param 2=1
    Module.Param0WriteEnabled=true
    Module.Param1WriteEnabled=true
    Module.Param2WriteEnabled=true
    Module.Param3WriteEnabled=true
   )
  ]
  particleUpdateScript [
   &ParticleState()
   &SkeletalMeshLocation(
    Bone / Socket Index=11
    Bone Sampling Mode=::NewEnumerator4
    MaskSpawnGroup=false
    Orientation Sampling=::NewEnumerator1
    Skeletal Mesh=$System.Skeletal Mesh
   )
   &ScaleColor(
    Scale Alpha=&Lerp_Float(A=1, B=0, Alpha=$Particles.NormalizedAge)
    Scale RGB=&VectorFromFloat(Value=&FN_FrontEndFloatModifier(Gameplay=1.5, Front End=1))
    ScaleA=true
    ScaleRGB=true
   )
   &FN_Light_Attributes(
    Color=$Particles.Color
    Create Color Attribute=true
    Create Exponent Attribute=true
    Create Position Attribute=true
    Create Radius Attribute=true
    Exponent=3
    Position=<15 0 0>
    Radius=200
    Scale Color=&FloatFromCurve(
     FloatCurve=DataInterface() {propertyValues {Curve {keys [
         {arriveTangent=6.6537, interpMode=RCIM_Cubic, leaveTangent=6.6537, tangentMode=RCTM_User, time=0, value=0}
         {interpMode=RCIM_Cubic, tangentMode=RCTM_User, time=0.5, value=1}
         {tangentMode=RCTM_User, time=1, value=0}
        ]}}}
     Scale Curve=5
    )
   )
  ]
  rendererClasses [&NiagaraLightRendererProperties]
  renderers [{rendererClass=&NiagaraLightRendererProperties, rendererIndex=0}]
 }
}
summary {
 systemName=NS_Emote_PlantStand_Fire
 emitters [
  {emitterName=Fire, rendererClasses [&NiagaraRibbonRendererProperties]}
  {emitterName=Fire_Sparks, rendererClasses [&NiagaraSpriteRendererProperties]}
  {emitterName=FlameSprites, rendererClasses [&NiagaraSpriteRendererProperties]}
  {emitterName=Light_for_Fire, rendererClasses [&NiagaraLightRendererProperties]}
 ]
 userVariables [
  User.bIsFrontEnd {type=NiagaraBool}
  User.bIsFrontEndPreview {type=NiagaraBool}
 ]
}
