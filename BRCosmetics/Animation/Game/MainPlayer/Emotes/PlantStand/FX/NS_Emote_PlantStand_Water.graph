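# Niagara water effect for the PlantStand emote, a brief stream emitted from a sampled skeletal mesh bone or socket.
# A ribbon, width-shaped over emitter age, forms the stream, while delayed blob and spark sprites add splashes with inherited velocity, curl noise and drag.
# Spinning flow meshes move along sampled velocity by setting position from a curve instead of forces.
# Blobs and sparks compute separate glow size and color alongside their dual sprite renderers.
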NiagaraSystem /BRCosmetics/Animation/Game/MainPlayer/Emotes/PlantStand/FX/NS_Emote_PlantStand_Water
dependencies {
 usedDataInterfaces [
  &NiagaraDataInterfaceCurve
  &NiagaraDataInterfaceVector2DCurve
  &NiagaraDataInterfaceArrayInt32
  &NiagaraDataInterfaceSkeletalMesh
 ]
 usedDynamicInputs []
 usedModules [
  &self:SystemSpawnScript.SystemScriptSource.SystemScriptGraph.NiagaraNodeAssignment_0.SetVariables_#1
  &SystemState
  &SampleSkeletalMesh
 ]
}
emitters {
 Blobs {
  emitterUpdateScript [
   &EmitterState(Life Cycle Mode=::NewEnumerator1, Loop Behavior=::NewEnumerator1, Loop Delay=0.2, Loop Duration=0.6, UseLoopDelay=true)
   &SpawnRate(Spawn Probability=0.8, SpawnRate=50, Use Spawn Probability=true)
  ]
  particleSpawnScript [
   &InitializeParticle(
    Color=&RandomRangeLinearColor(Minimum=<rgba 0 0.1854 0.6823 0.5>, Maximum=<rgba 0.2061 1.0923 1.2 0.3>)
    Lifetime Max=0.6
    Lifetime Min=0.4
    Lifetime Mode=::NewEnumerator1
    Mass Mode=::NewEnumerator3
    Sprite Size Mode=::NewEnumerator1
    Uniform Sprite Size Max=15
    Uniform Sprite Size Min=10
    UsePositionOffset=false
    Write Material Random=true
   )
   &SkeletalMeshLocation(
    Bone / Socket Index=2
    Bone Sampling Mode=::NewEnumerator4
    "Inherit Velocity (Scale)":NiagaraFloat=&RandomRangeFloat(Minimum=0.5, Maximum=1)
    Inherited Velocity Speed Limit=100
    MaskSpawnGroup=false
    Orientation Sampling=::NewEnumerator1
    Skeletal Mesh=$System.Skeletal Mesh
    Velocity Sampling=::NewEnumerator0
    VelocitySpeedLimit=true
   )
   &ShapeLocation(MaskSpawnGroup=true, Override Seed=false, Spawn Group Mask=1, Sphere Radius=10, UseOwnerScale=true) {module=ShapeLocation001}
   &AddVelocity(
    Module.Use Falloff Distance=false
    Origin Offset=<0 0 -30>
    Origin Offset Coordinate Space=::World
    Override Seed=false
    Velocity Mode=::NewEnumerator1
    Velocity Speed=&RandomRangeFloat(Minimum=10, Maximum=50)
   )
  ]
  particleUpdateScript [
   &ParticleState()
   &ScaleSpriteSize(
    Non-Uniform Curve Scale=<1 1.5>
    Non-Uniform Curve Sprite Scale=DataInterface() {propertyValues {
      XCurve {keys [
        {interpMode=RCIM_Cubic, tangentMode=RCTM_User, time=0, value=0}
        {interpMode=RCIM_Cubic, tangentMode=RCTM_User, time=0.5, value=1}
        {tangentMode=RCTM_User, time=1, value=0}
       ]}
      YCurve {keys [
        {arriveTangent=2.5642, interpMode=RCIM_Cubic, leaveTangent=2.5642, tangentMode=RCTM_User, time=0, value=0.75}
        {interpMode=RCIM_Cubic, tangentMode=RCTM_User, time=0.5, value=1.575}
        {arriveTangent=-1.2758, interpMode=RCIM_Cubic, leaveTangent=-1.2758, tangentMode=RCTM_User, time=1, value=1.25}
       ]}
     }}
    Scale Sprite Size Mode=::NewEnumerator3
   )
   &Drag(Drag=5, Use Linear Drag=true, Use Rotational Drag=false)
   &ScaleColor(
    Scale Alpha=&FloatFromCurve(FloatCurve=DataInterface() {propertyValues {Curve {keys [<at -0.0046 1.0113 RCIM_Cubic>,<at 1.0011 0.002 RCIM_Cubic>]}}})
    ScaleA=true
    ScaleRGB=true
   )
   &CurlNoiseForce(Noise Frequency=20, Noise Strength=50, Pan Noise=true, Pan Noise Field=<0 0 1>)
   &SolveForcesAndVelocity(Clamp Velocity=false, Limit Acceleration=false)
   &ScaleSpriteSizeBySpeed(Max Scale Factor=<1 3>, Velocity Threshold=300)
   &DynamicMaterialParameters(
    Index 0 Param 1=&FloatFromCurve(
     FloatCurve=DataInterface() {propertyValues {Curve {keys [{interpMode=RCIM_Cubic, tangentMode=RCTM_User, time=0, value=1}, {interpMode=RCIM_Cubic, tangentMode=RCTM_User, time=1, value=0.35}]}}}
     Scale Curve=2
    )
    Module.Index0Param0Write=true
    Module.Index0Param1Write=true
    Module.Index0Param2Write=true
    Module.Index0Param3Write=true
   )
   /SetVariables(Particles.SpriteSizeGlow:Vector2f=&Multiply_Vector2D_ByFloat(Float=1.25, Vector2D=$Particles.SpriteSize)) {isSetParameters=true}
   /SetVariables(Particles.Glow Color=&MultiplyLinearColor(Linear Color=$Particles.Color, Scale Factor=<1 1 1 0.15>)) {isSetParameters=true}
  ]
  rendererClasses [&NiagaraSpriteRendererProperties]
  renderers [
   {rendererClass=&NiagaraSpriteRendererProperties}
   {rendererClass=&NiagaraSpriteRendererProperties, rendererIndex=1}
  ]
 }
 Water {
  emitterUpdateScript [
   &EmitterState(Life Cycle Mode=::NewEnumerator1, Loop Behavior=::NewEnumerator1, Loop Duration=1.2, UseLoopDelay=true)
   &SpawnRate(SpawnRate=60, Use Spawn Probability=false)
  ]
  particleSpawnScript [
   &InitializeParticle(
    Color=<rgba 0 0.6452 2 1>
    Lifetime=0.55
    Ribbon Width=&FloatFromCurve(
     FloatCurve=DataInterface() {propertyValues {Curve {keys [
         {arriveTangent=8.4015, interpMode=RCIM_Cubic, leaveTangent=8.4015, tangentMode=RCTM_User, time=0.15, value=0}
         {arriveTangent=-0.0151, interpMode=RCIM_Cubic, leaveTangent=-0.0151, tangentMode=RCTM_User, time=0.3, value=1}
         {arriveTangent=-0.0972, interpMode=RCIM_Cubic, leaveTangent=-0.0972, tangentMode=RCTM_User, time=0.7, value=1}
         {arriveTangent=-11.0142, interpMode=RCIM_Cubic, leaveTangent=-11.0142, tangentMode=RCTM_User, time=0.9, value=0.025}
        ]}}}
     CurveIndex=$Emitter.Age
     Scale Curve=35
    )
    Ribbon Width Mode=::NewEnumerator1
    UsePositionOffset=false
    Write Material Random=true
   ) {module=InitializeParticle001}
   &SkeletalMeshLocation(
    Bone / Socket Index=2
    Bone Sampling Mode=::NewEnumerator4
    Interpolate Mesh Sampling=false
    MaskSpawnGroup=false
    Orientation Sampling=::NewEnumerator1
    Skeletal Mesh=$System.Skeletal Mesh
   )
   &DynamicMaterialParameters(
    Index 0 Param 3=20
    Index 0 Param 4=0.2
    Module.Index0Param0Write=true
    Module.Index0Param1Write=true
    Module.Index0Param2Write=true
    Module.Index0Param3Write=true
   )
   /SetVariables(Particles.Linear Color=&FN_Select_Color_from_Time_Of_Day(Morning=<rgba 1 1 1 1>, Day=<rgba 1 1 1 1>, Evening=<rgba 1 1 1 1>, Night=<rgba 0.75 0.75 0.75 1>)) {isSetParameters=true}
  ]
  particleUpdateScript [
   &ParticleState()
   &ScaleColor(Scale Alpha=0.5, ScaleA=true, ScaleRGB=true)
   &ScaleRibbonWidth(Ribbon Width Scale=&FloatFromCurve(FloatCurve=DataInterface() {propertyValues {Curve {keys [
         {arriveTangent=7.2592, interpMode=RCIM_Cubic, leaveTangent=7.2592, tangentMode=RCTM_User, time=0, value=0}
         {arriveTangent=0.0117, interpMode=RCIM_Cubic, leaveTangent=0.0117, tangentMode=RCTM_User, time=0.4, value=1.2}
         {arriveTangent=-0.0405, interpMode=RCIM_Cubic, leaveTangent=-0.0405, tangentMode=RCTM_User, time=1, value=0}
        ]}}}))
   &CurlNoiseForce(Noise Frequency=10, Pan Noise=true, Pan Noise Field=<0 0 0.1>)
   &SolveForcesAndVelocity(Clamp Velocity=false, Limit Acceleration=false)
  ]
  rendererClasses [&NiagaraRibbonRendererProperties]
  renderers [
   {rendererClass=&NiagaraRibbonRendererProperties}
   {rendererClass=&NiagaraRibbonRendererProperties, rendererIndex=1}
  ]
 }
 Water_Flow {
  emitterUpdateScript [
   &EmitterState(Life Cycle Mode=::NewEnumerator1, Loop Behavior=::NewEnumerator1, Loop Delay=0.2, Loop Duration=0.6, UseLoopDelay=true) {module=EmitterState001}
   &SpawnRate(Spawn Probability=0.8, SpawnRate=35, Use Spawn Probability=true)
  ]
  particleSpawnScript [
   &InitializeParticle(
    Color Maximum=<rgba 0.063 0.3179 0.5 0.3>
    Color Minimum=<rgba 0.224 1.1712 2 0.5>
    Color Mode=::NewEnumerator2
    Lifetime Max=0.8
    Lifetime Min=0.5
    Lifetime Mode=::NewEnumerator1
    Mesh Scale=&Multiply_VectorByFloat(
     Vector=&RandomRangeVector(Minimum=<1 1 2>, Maximum=<1 1 3>)
     Float=&RandomRangeFloat(Minimum=0.208, Maximum=0.39)
    )
    Mesh Scale Mode=::NewEnumerator4
    UsePositionOffset=false
    Write Material Random=true
   )
   &SkeletalMeshLocation(
    Bone / Socket Index=2
    Bone Sampling Mode=::NewEnumerator4
    "Inherit Velocity (Scale)":NiagaraFloat=&RandomRangeFloat(Minimum=0.5, Maximum=1)
    Inherited Velocity Speed Limit=10
    MaskSpawnGroup=false
    Skeletal Mesh=$System.Skeletal Mesh
    Velocity Sampling=::NewEnumerator0
    VelocitySpeedLimit=true
   )
   &InitialMeshOrientation(
    Mesh Axis To Orient=<0 0 1>
    Mesh Orientation Mode=::NewEnumerator2
    Rotation=&Multiply_VectorByFloat(
     Vector=&RandomRangeVector(Minimum=<0 0 0>, Maximum=<0.1 0.1 1>)
     Float=&MakeFloatFromInt(Int=&SelectIntFromArray(
       Array Sampling Mode=::NewEnumerator1
       Direct Array Index=?
       Int Selection Array=DataInterface() {propertyValues {IntData [-1,1]}}
       Random Seed=0
      ))
    )
    Use Rotation Vector=true
   )
  ]
  particleUpdateScript [
   &ParticleState()
   &DynamicMaterialParameters(
    Index 0 Param 1=&FloatFromCurve(FloatCurve=DataInterface() {propertyValues {Curve {keys [
         {interpMode=RCIM_Cubic, tangentMode=RCTM_User, time=0, value=1}
         <at 0.7 0.35>
         {interpMode=RCIM_Cubic, tangentMode=RCTM_User, time=1, value=0}
        ]}}})
    Module.Param0WriteEnabled=true
    Module.Param1WriteEnabled=true
    Module.Param2WriteEnabled=true
    Module.Param3WriteEnabled=true
   )
   &ScaleMeshSize(Scale Factor=&Multiply_VectorByFloat(
     Vector=&Multiply_VectorByFloat(Vector=$Engine.Owner.Scale, Float=1)
     Float=&FloatFromCurve(
      FloatCurve=DataInterface() {propertyValues {Curve {keys [
          {arriveTangent=2.0131, interpMode=RCIM_Cubic, leaveTangent=2.0131, tangentMode=RCTM_User, time=0, value=0}
          {arriveTangent=0.0166, interpMode=RCIM_Cubic, leaveTangent=0.0166, tangentMode=RCTM_User, time=0.7, value=1}
         ]}}}
      Scale Curve=&FN_FrontEndFloatModifier(Gameplay=1, Front End=1)
     )
    ))
   &UpdateMeshOrientation(Rotation Rate=&RandomRangeFloat(Minimum=0.5, Maximum=1.5), Rotation Vector=<0 0 1>)
   &ScaleColor(
    Scale Alpha=&FloatFromCurve(FloatCurve=DataInterface() {propertyValues {Curve {keys [
         {arriveTangent=4.4696, interpMode=RCIM_Cubic, leaveTangent=4.4696, tangentMode=RCTM_User, time=0, value=0}
         {interpMode=RCIM_Cubic, tangentMode=RCTM_User, time=0.7, value=1}
         {arriveTangent=-5.2232, tangentMode=RCTM_User, time=1, value=0}
        ]}}})
    ScaleA=true
    ScaleRGB=true
   )
   /SetVariables(Particles.Position=&AddVectorToPosition(
     Position=$Particles.Initial.Position
     Vector=&Multiply_VectorByFloat(
      Vector=&Add_Vector(
       A=&NormalizeVector(
        Vector To Normalize=$Particles.SkeletalMeshLocation.SampledVelocity
        Safe Normalize=true
        Safe Fallback Threshold=0
        Safe Fallback Vector=<0 0 0.001>
       )
       B=&RandomVector(Vector Scale=0.1)
      )
      Float=&FloatFromCurve(
       FloatCurve=DataInterface() {propertyValues {Curve {keys [
           {arriveTangent=2.4228, interpMode=RCIM_Cubic, leaveTangent=2.4228, tangentMode=RCTM_User, time=0, value=0}
           {arriveTangent=0.0224, interpMode=RCIM_Cubic, leaveTangent=0.0224, tangentMode=RCTM_User, time=0.8, value=1}
          ]}}}
       Scale Curve=&RandomRangeFloat(Minimum=10, Maximum=20)
      )
     )
    )) {isSetParameters=true}
  ]
  rendererClasses [&NiagaraMeshRendererProperties]
  renderers [{rendererClass=&NiagaraMeshRendererProperties, rendererIndex=0}]
 }
 Water_Sparks {
  emitterUpdateScript [
   &EmitterState(Life Cycle Mode=::NewEnumerator1, Loop Behavior=::NewEnumerator1, Loop Delay=0.2, Loop Duration=0.6, UseLoopDelay=true)
   &SpawnRate(Spawn Probability=0.8, SpawnRate=75, Use Spawn Probability=true)
   &SpawnRate(Spawn Probability=0.8, SpawnRate=100, Use Spawn Probability=true) {module=SpawnRate001}
  ]
  particleSpawnScript [
   &InitializeParticle(
    Color Maximum=<rgba 0 0.6405 1 1>
    Color Minimum=<rgba 0 0.1524 0.3646 1>
    Color Mode=::NewEnumerator2
    Lifetime=&RandomRangeFloat(Minimum=0.3, Maximum=0.8)
    Sprite Size Max=<3 8>
    Sprite Size Min=<1 4>
    Sprite Size Mode=::NewEnumerator2
    UsePositionOffset=false
    Write Material Random=true
   )
   &SkeletalMeshLocation(
    Bone / Socket Index=2
    Bone Sampling Mode=::NewEnumerator4
    "Inherit Velocity (Scale)":NiagaraFloat=&RandomRangeFloat(Minimum=0.5, Maximum=1)
    Inherited Velocity Speed Limit=150
    MaskSpawnGroup=false
    Orientation Sampling=::NewEnumerator1
    Skeletal Mesh=$System.Skeletal Mesh
    Velocity Sampling=::NewEnumerator0
    VelocitySpeedLimit=true
   )
   /SetVariables(Particles.RandomCore:NiagaraFloat=&FloatFromCurve(
     FloatCurve=DataInterface() {propertyValues {Curve {keys [
         {arriveTangent=0.0559, interpMode=RCIM_Cubic, leaveTangent=0.0559, tangentMode=RCTM_User, time=0, value=0.2}
         {arriveTangent=0.1895, interpMode=RCIM_Cubic, leaveTangent=0.1895, tangentMode=RCTM_User, time=0.9, value=0.25}
         {arriveTangent=1.1808, interpMode=RCIM_Cubic, leaveTangent=1.1808, tangentMode=RCTM_User, time=1, value=0.3}
        ]}}}
     CurveIndex=&RandomRangeFloat(Minimum=0, Maximum=1)
     Scale Curve=2
    )) {isSetParameters=true}
   &ShapeLocation(MaskSpawnGroup=false, Override Seed=false, Sphere Radius=5, UseOwnerScale=true)
  ]
  particleUpdateScript [
   &ParticleState()
   &CurlNoiseForce(Noise Frequency=10, Noise Strength=150, Pan Noise=true, Pan Noise Field=<0 0 1>)
   &ScaleSpriteSize(Scale Factor=&Vector2DFromFloat(Value=&FloatFromCurve(FloatCurve=DataInterface() {propertyValues {Curve {keys [
          {arriveTangent=9.5944, interpMode=RCIM_Cubic, leaveTangent=9.5944, tangentMode=RCTM_User, time=0, value=0}
          <at 0.2 1.2 RCIM_Cubic>
          {arriveTangent=-2.7132, interpMode=RCIM_Cubic, leaveTangent=-2.7132, tangentMode=RCTM_User, time=1, value=0}
         ]}}})))
   /SetVariables(Particles.Glow03_Size=&Multiply_Vector2D(A=$Particles.SpriteSize, B=<4 3>)) {isSetParameters=true}
   &ScaleColor(
    Scale RGB=&Multiply_VectorByFloat(Vector=<2 2 2>, Float=&FNTimeOfDayScalar(Night Scale=1, Morning Scale=0.7, Evening Scale=0.7))
    ScaleA=true
    ScaleRGB=true
   )
   /SetVariables(Particles.Glow03_Color=&MultiplyLinearColor(
     Linear Color=$Particles.Color
     Scale Factor=&MakeFloat4FromVectorAndFloat("Vector (RGB)"=<1.5 1.5 1.5>, "Float (Alpha)"=0.1)
    )) {isSetParameters=true}
   &Drag(Drag=5, Use Linear Drag=true, Use Rotational Drag=false)
   &SolveForcesAndVelocity(Clamp Velocity=false, Limit Acceleration=false)
  ]
  rendererClasses [&NiagaraSpriteRendererProperties]
  renderers [
   {rendererClass=&NiagaraSpriteRendererProperties}
   {rendererClass=&NiagaraSpriteRendererProperties, rendererIndex=1}
  ]
 }
}
summary {
 systemName=NS_Emote_PlantStand_Water
 emitters [
  {emitterName=Water, rendererClasses [&NiagaraRibbonRendererProperties]}
  {emitterName=Water_Sparks, rendererClasses [&NiagaraSpriteRendererProperties]}
  {emitterName=Blobs, rendererClasses [&NiagaraSpriteRendererProperties]}
  {emitterName=Water_Flow, rendererClasses [&NiagaraMeshRendererProperties]}
 ]
 userVariables [
  User.bIsFrontEnd {type=NiagaraBool}
  User.bIsFrontEndPreview {type=NiagaraBool}
 ]
}
